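# Unlit, two-sided masked material for the Ninja Wolf cosmetic's spike effect on Niagara mesh particles, drawing animated anime-style flames.
# Panning fire-mask samples, distorted by a normal map, vertex red and a sine-wave mask, combine into three summed layers.
# That shape indexes a tinted gradient with time-of-day HDR scaling and, faded by a sine across U, forms the mask.
# A time-driven sine, varied by UV1, pulses vertices around each particle pivot.
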Material /BRCosmetics/Characters/Player/Male/Medium/Bodies/M_MED_Ninja_Wolf/Effects/M_NinjaWolf_Spikes
graph {
 properties {
  bCastRayTracedShadows=false
  bHasPixelAnimation=true
  bUsedWithNiagaraMeshParticles=true
  blendMode=BLEND_Masked
  pixelDepthOffsetMode=PDOM_Legacy
  previewMesh=&/hairdo
  shadingModel=MSM_Unlit
  twoSided=true
  usedShadingModels=Unlit
 }
 nodes [
  n1 = VertexColor()
  n2 = ScalarParameter() {defaultValue=0.2, parameterName=Vert_R_Scale}
  n3 = Add(A=TextureCoordinate(), B=AppendVector(A=Multiply(A=@n1.R, B=@n2), B=Constant()))
  n4 = TextureSample(UVs=Panner(Coordinate=TextureCoordinate()) {speedX=-1, speedY=0.1}) {
   samplerSource=SSM_Wrap_WorldGroupSettings
   samplerType=SAMPLERTYPE_Normal
   texture=/Game/Effects/Fort_Effects/Gameplay/SmokeDustExplosions/Textures/T_Explosion_NM_01
  }
  n5 = ComponentMask(None=MaterialFunctionCall(Normal=@n4.RGB, Flatness=ScalarParameter() {defaultValue=1, parameterName=Normal_Scale}) {materialFunction=&FlattenNormal}.Result) {g=true, r=true}
  n6 = Add(A=@n3, B=@n5)
  n7 = ComponentMask(None=@n6) {r=true}
  n8 = Add(A=ComponentMask(None=@n6) {g=true}) {constB=-0.15}
  n9 = ComponentMask(None=@n6) {r=true}
  n10 = Add(A=ComponentMask(None=@n6) {g=true}) {constB=0.45}
  n11 = VectorParameter() {group=AnimeFlameShape, parameterName=Anime Flame Shape Panner Speed, sortPriority=3, defaultValue=<rgba 3 0 0 0>}
  n12 = ComponentMask(None=@n11.RGB) {g=true, r=true}
  n13 = Panner(Coordinate=LinearInterpolate(A=AppendVector(A=@n7, B=@n8), B=AppendVector(A=@n9, B=@n10), Alpha=NamedRerouteUsage()), Speed=@n12)
  n14 = ScalarParameter() {parameterName=Bias_A}
  n15 = Add(A=@n3, B=@n5)
  n16 = ComponentMask(None=@n15) {r=true}
  n17 = Add(A=ComponentMask(None=@n15) {g=true}) {constB=-0.15}
  n18 = ComponentMask(None=@n15) {r=true}
  n19 = Add(A=ComponentMask(None=@n15) {g=true}) {constB=0.45}
  n20 = Panner(Coordinate=LinearInterpolate(A=AppendVector(A=@n16, B=@n17), B=AppendVector(A=@n18, B=@n19), Alpha=NamedRerouteUsage()), Speed=@n12)
  n21 = Multiply(A=Multiply(
    A=TextureSample!(UVs=@n13, Tex, Bias=@n14, Apply View MipBias) {mipValueMode=TMVM_MipBias, texture=/Game/Effects/Fort_Effects/Textures/Fire/T_Fire_Mask}.R
    B=TextureSample(UVs=@n20, Bias=@n14) {mipValueMode=TMVM_MipBias, texture=&T_Fire_Mask}.R
   )) {constB=0.333}
  n22 = Add(A=@n3, B=@n5)
  n23 = ComponentMask(None=@n22) {r=true}
  n24 = Add(A=ComponentMask(None=@n22) {g=true}) {constB=-0.15}
  n25 = ComponentMask(None=@n22) {r=true}
  n26 = Add(A=ComponentMask(None=@n22) {g=true}) {constB=0.25}
  n27 = Panner(
   Coordinate=LinearInterpolate(A=AppendVector(A=@n23, B=@n24), B=AppendVector(A=@n25, B=@n26), Alpha=NamedRerouteUsage())
   Speed=NamedRerouteUsage()
  )
  n28 = ScalarParameter() {parameterName=Bias_A}
  n29 = Add(A=@n3, B=@n5)
  n30 = ComponentMask(None=@n29) {r=true}
  n31 = Add(A=ComponentMask(None=@n29) {g=true}) {constB=-0.15}
  n32 = ComponentMask(None=@n29) {r=true}
  n33 = Add(A=ComponentMask(None=@n29) {g=true}) {constB=0.25}
  n34 = Panner(
   Coordinate=LinearInterpolate(A=AppendVector(A=@n30, B=@n31), B=AppendVector(A=@n32, B=@n33), Alpha=NamedRerouteUsage())
   Speed=NamedRerouteUsage()
  )
  n35 = Multiply(A=Multiply(
    A=TextureSample(UVs=@n27, Bias=@n28) {mipValueMode=TMVM_MipBias, texture=&T_Fire_Mask}.R
    B=TextureSample(UVs=@n34, Bias=@n28) {mipValueMode=TMVM_MipBias, texture=&T_Fire_Mask}.R
   )) {constB=0.333}
  n36 = Add(A=@n3, B=@n5)
  n37 = ComponentMask(None=@n36) {g=true}
  n38 = ComponentMask(None=@n36) {g=true}
  n39 = LinearInterpolate(
   A=AppendVector(A=ComponentMask(None=@n36) {r=true}, B=Add(A=@n37) {constB=-0.15})
   B=AppendVector(A=ComponentMask(None=@n36) {r=true}, B=Add(A=@n38) {constB=-0.05})
   Alpha=NamedRerouteUsage()
  )
  n40 = NamedRerouteUsage()
  n41 = ScalarParameter() {parameterName=Bias_A}
  n42 = Add(A=@n3, B=@n5)
  n43 = ComponentMask(None=@n42) {g=true}
  n44 = ComponentMask(None=@n42) {g=true}
  n45 = TextureCoordinate() {uTiling=2}
  n46 = Multiply(A=VertexColor().R, B=ScalarParameter() {defaultValue=0.2, parameterName=Vert_R_Scale})
  n47 = VectorParameter() {group=AnimeFlameShape, parameterName=Anime Flame SineMask Panner Speed, sortPriority=3, defaultValue=<rgba 0 0 1 0>}
  n48 = TextureSample!(UVs=Panner(Coordinate=Add(A=@n45, B=@n46), Speed=ComponentMask(None=@n47.RGB) {g=true, r=true}), Tex, Apply View MipBias) {samplerType=SAMPLERTYPE_Masks, texture=/Engine/Functions/Engine_MaterialFunctions02/ExampleContent/Textures/SineWaveHeightMap}
  n49 = LinearInterpolate(
   A=AppendVector(A=ComponentMask(None=@n42) {r=true}, B=Add(A=@n43) {constB=-0.15})
   B=AppendVector(A=ComponentMask(None=@n42) {r=true}, B=Add(A=@n44) {constB=-0.05})
   Alpha=NamedRerouteDeclaration(None=LinearInterpolate(Alpha=@n48.R) {constA=0.5, constB=1.5}) {name=FlameSin Lerp Alpha, nodeColor=<rgba 0 1 0.5547 1>}
  )
  n50 = VectorParameter() {group=AnimeFlameShape, parameterName=Anime Flame Shape Panner Speed, sortPriority=3, defaultValue=<rgba 3 0 0 0>}
  n51 = NamedRerouteDeclaration(None=ComponentMask(None=@n50.RGB) {g=true, r=true}) {name=Flame Panner Speed, nodeColor=<rgba 0.2422 1 0 1>}
  n52 = Multiply(
   A=TextureSample!(UVs=Panner(Coordinate=@n39, Speed=@n40), Tex, Bias=@n41, Apply View MipBias) {mipValueMode=TMVM_MipBias, texture=&T_Fire_Mask}.R
   B=TextureSample(UVs=Panner(Coordinate=@n49, Speed=@n51), Bias=@n41) {mipValueMode=TMVM_MipBias, texture=&T_Fire_Mask}.R
  )
  n53 = Saturate(None=Add(A=Add(A=@n21, B=@n35), B=Multiply(A=@n52) {constB=0.333}))
  n54 = ScalarParameter() {defaultValue=2.25, group=Color, parameterName=Curve Contrast Controller, sortPriority=1}
  n55 = Multiply(
   A=TextureSample!(
    UVs=Power(Base=@n53, Exp=@n54)
    Tex=TextureObjectParameter() {parameterName=Gradient, texture=/Game/Effects/Materials/SwingTrails/Textures/T_FX_ColorRamp_Linear}
    Apply View MipBias
   ) {samplerSource=SSM_Clamp_WorldGroupSettings}.RGB
   B=VectorParameter() {group=Color, parameterName=HDR_Tint, sortPriority=2, defaultValue=<rgba 2 1 1.5 0>}.RGB
  )
  n56 = ScalarParameter() {defaultValue=0.15, parameterName=HDR_Morning}
  n57 = ScalarParameter() {defaultValue=2, parameterName=HDR_Day}
  n58 = ScalarParameter() {defaultValue=0.7, parameterName=HDR_Evening}
  n59 = ScalarParameter() {defaultValue=1, parameterName=HDR_Night}
  MaterialFunctionCall(Emissive=MaterialFunctionCall(
    LDR Input=Clamp(None=@n55)
    HDR Tint=@n55
    HDR Intensity Multiplier=MaterialFunctionCall(Value_Morning=@n56, Value_Day=@n57, Value_Evening=@n58, Value_Night=@n59) {materialFunction=&Scalar_Time_of_Day}.Result
    Derive HDR Power=ScalarParameter() {defaultValue=1, group=Color, parameterName=Derive HDR Power, sortPriority=3}
   ) {materialFunction=&DeriveHDRfromLDR}.RemappedRange) {materialFunction=&MF_FX_EmissiveAdjustments}
  n60 = Subtract(
   A=MaterialFunctionCall() {materialFunction=&ObjectPivotPoint}.Mesh Particle Pivot Location
   B=WorldPosition().XYZ
  )
  n61 = Multiply(
   A=Multiply(A=Time(), B=ScalarParameter() {defaultValue=200, parameterName=timeSpeed})
   B=ComponentMask(None=TextureCoordinate() {coordinateIndex=1, vTiling=0}) {r=true}
  )
  n62 = LinearInterpolate(
   A=ScalarParameter() {defaultValue=-0.8, parameterName=minLerp}
   B=ScalarParameter() {defaultValue=0.5, parameterName=maxLerp}
   Alpha=ConstantBiasScale(None=Sine(None=@n61))
  ) {constA=-10, constB=0.5}
  NamedRerouteDeclaration(None=Multiply(A=ScalarParameter() {defaultValue=0.05, parameterName=WPOIntensity}, B=Multiply(A=@n60, B=@n62))) {name=WPO, nodeColor=<rgba 0.8516 1 0 1>}
  NamedRerouteUsage()
  n63 = Sine(None=ComponentMask(None=TextureCoordinate()) {r=true}) {period=2}
  Saturate(None=Multiply(A=ScalarParameter() {defaultValue=1, group=Alpha, parameterName=MasterAlpha}, B=Multiply(A=@n53, B=@n63)))
 ]
}
