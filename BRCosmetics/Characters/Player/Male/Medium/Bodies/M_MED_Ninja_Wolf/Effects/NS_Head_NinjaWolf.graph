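# Niagara head effect for the Ninja Wolf character cosmetic, with flame sprites and a glow light attached to the skeletal mesh.
# EyeFlames continuously spawns short-lived GPU sprites that inherit source movement and drift with curl noise.
# Flash burst-spawns a mesh particle that tracks the skeleton and carries the light.
# BigHeadMode scales both emitters, IsShadow switches flame gradient and mesh visibility tag, and bisHeadHidden drives visibility tags for hiding.
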NiagaraSystem /BRCosmetics/Characters/Player/Male/Medium/Bodies/M_MED_Ninja_Wolf/Effects/NS_Head_NinjaWolf
dependencies {
 usedDataInterfaces [
  &NiagaraDataInterfaceArrayInt32
  &NiagaraDataInterfaceCurve
  &NiagaraDataInterfaceVectorCurve
  &NiagaraDataInterfaceColorCurve
  &NiagaraDataInterfaceSkeletalMesh
 ]
 usedDynamicInputs [&VectorFromCurve]
 usedModules [
  &self:SystemSpawnScript.SystemScriptSource.SystemScriptGraph.NiagaraNodeAssignment_4.SetVariables_#1
  &self:SystemSpawnScript.SystemScriptSource.SystemScriptGraph.NiagaraNodeAssignment_0.SetVariables_#2
  &self:SystemSpawnScript.SystemScriptSource.SystemScriptGraph.NiagaraNodeAssignment_1.SetVariables_#3
  &self:SystemSpawnScript.SystemScriptSource.SystemScriptGraph.NiagaraNodeAssignment_2.SetVariables_#4
  &self:SystemSpawnScript.SystemScriptSource.SystemScriptGraph.NiagaraNodeAssignment_3.SetVariables_#5
  &SystemState
 ]
}
emitters {
 EyeFlames {
  simTarget=GPUComputeSim
  emitterUpdateScript [
   &EmitterState(Life Cycle Mode=::NewEnumerator1, Loop Duration=1, UseLoopDelay=true)
   &SpawnRate(SpawnRate=50, Use Spawn Probability=false)
  ]
  particleSpawnScript [
   /Niagara/Modules/Spawn/Initialization/InitializeParticle(
    Color=<rgba 0.7 0.7 0.7 1>
    Lifetime=&UniformRangedFloat(Minimum=0.2, Maximum=0.5)
    Module.Write Color=true
    Module.Write Lifetime=true
    Module.Write Mass=false
    Module.Write Material Random=true
    Module.Write Position=false
    Module.Write Scale=false
    Module.Write SpriteRotation=true
    Module.Write SpriteSize=true
    Position=&SimulationPosition()
    Sprite Rotation=&UniformRangedFloat(Minimum=0, Maximum=360)
    Sprite Size=&Vector2DFromFloat(Value=&RandomRangeFloat(
      Minimum=&Lerp_Float(A=0.5, B=1.2, Alpha=$NPC.FortniteNPC.BigHeadMode)
      Maximum=&Lerp_Float(A=2, B=3, Alpha=$NPC.FortniteNPC.BigHeadMode)
     ))
   ) {enabled=false}
   /Niagara/Modules/Spawn/Initialization/V2/InitializeParticle(
    Color=<rgba 0.7 0.7 0.7 1>
    Lifetime=&UniformRangedFloat(Minimum=0.2, Maximum=0.5)
    Sprite Rotation Mode=::NewEnumerator2
    Sprite Size=&Vector2DFromFloat(Value=&RandomRangeFloat(
      Minimum=&Lerp_Float(A=0.5, B=1.2, Alpha=$NPC.FortniteNPC.BigHeadMode)
      Maximum=&Lerp_Float(A=2, B=3, Alpha=$NPC.FortniteNPC.BigHeadMode)
     ))
    Sprite Size Mode=::NewEnumerator4
    UsePositionOffset=false
    Write Material Random=true
   ) {module=InitializeParticle001}
   &DynamicMaterialParameters(
    Index 0 Param 2=&UniformRangedFloat(Minimum=0, Maximum=1)
    Index 0 Param 3=1
    Index 0 Param 4=&UniformRangedFloat(Minimum=1, Maximum=1)
    Module.Param0WriteEnabled=false
    Module.Param1WriteEnabled=true
    Module.Param2WriteEnabled=false
    Module.Param3WriteEnabled=true
   ) {module=DynamicMaterialParameters002}
   &FN_SetSpawnGroup(Number of Spawn Groups=2)
   &CameraOffset(Camera Offset Amount=&RandomRangeFloat(Minimum=5, Maximum=7))
   &SkeletalMeshLocation(
    Apply Sampled Orientation to Mesh Particles=true
    Inherit Velocity=false
    Interpolate Mesh Sampling=false
    MaskSpawnGroup=false
    Mesh Sampling Type=::NewEnumerator5
    Sampled Position Offset=<0 6 -10>
    Sampled Position Offset Coordinate Space:ENiagaraMeshTransforms=::NewEnumerator3
    Skeletal Mesh=$System.Skeletal Mesh Reader
    Socket Sampling Mode=::NewEnumerator5
   ) {module=SkeletalMeshLocation001}
   &SphereLocation(
    MaskSpawnGroup=false
    Offset=&Multiply_Vector(A=<7 -5 0>, B=&Lerp_Vector(A=<1 1 1>, B=<0.8 0.8 1>, Alpha=$NPC.FortniteNPC.BigHeadMode))
    Override Seed=false
    Sphere Radius=&Lerp_Float(A=10, B=14, Alpha=$NPC.FortniteNPC.BigHeadMode)
    Surface Only=false
    UseNonUniformScale=false
    UseOffset=true
    UseSphereOrientation=false
   )
   &AddVelocityInCone(Cone Angle=65, Cone Axis=<0 -1 0>, Module.Use Velocity Falloff On Cone Axis=false, Velocity Strength=100)
   &AddVelocity(Coordinate Space=::World, Velocity=&RandomRangeVector(Minimum=<0 0 0>, Maximum=<1 1 50>)) {module=AddVelocity001}
  ]
  particleUpdateScript [
   &InheritSourceMovement(
    Applied Position Scale=&VectorFromFloat(Value=&FloatFromCurve(FloatCurve=DataInterface() {propertyValues {Curve {keys [
          {arriveTangent=-0.1644, interpMode=RCIM_Cubic, time=0, value=1}
          {arriveTangent=-2.3188, interpMode=RCIM_Cubic, leaveTangent=-2.3188, tangentMode=RCTM_User, time=1, value=0}
         ]}}}))
    ApplyForce=false
    ApplyPosition=true
    ApplyPositionFalloff=false
   )
   &CurlNoiseForce(Noise Frequency=4000, Noise Quality / Cost=::NewEnumerator0, Noise Strength=50, Pan Noise=false)
   &Drag(Drag=4, Use Linear Drag=true, Use Rotational Drag=false)
   &SolveForcesAndVelocity(Clamp Velocity=false, Limit Acceleration=false)
   &ParticleState()
   &ScaleColor(
    Scale Mode=::NewEnumerator1
    Scale RGBA=/Niagara/DynamicInputs/TypeConversions/MakeCustomVector4FromBool(
     Bool:NiagaraBool=&SetBoolByFloatComparison(A=$User.IsShadow, B=1, Comparison Type=::NewEnumerator2)
     True Vector4:Vector4f=&MakeVector4_FromLinearColor(LinearColor=&ColorFromCurve(ColorCurve=DataInterface() {propertyValues {
         AlphaCurve {keys [<at 0 0>,<at 0.1154 1>,<at 1 0>]}
         BlueCurve {keys [<at 0.0072 0>,<at 0.3568 0>,<at 1 0>]}
         GreenCurve {keys [<at 0.0072 0.426>,<at 0.3568 0.0688>,<at 1 0.0019>]}
         RedCurve {keys [<at 0.0072 2>,<at 0.3568 2>,<at 1 0.0521>]}
        }}))
     False Vector4:Vector4f=&MakeVector4_FromLinearColor(LinearColor=&ColorFromCurve(ColorCurve=DataInterface() {propertyValues {
         AlphaCurve {keys [<at 0 0>,<at 0.1154 1>,<at 1 0>]}
         BlueCurve {keys [<at 0.0094 0.128>,<at 1 0.0521>]}
         GreenCurve {keys [<at 0.0094 0>,<at 1 0.006>]}
         RedCurve {keys [<at 0.0094 2>,<at 1 0.0295>]}
        }}))
    )
    ScaleRGBA=true
   )
   &SpriteRotationRate(Rotation Rate=100)
   &ScaleSpriteSizeBySpeed(Max Scale Factor=<0.75 2.5>, Min Scale Factor=<1 0.25>, Velocity Threshold=100)
   /SetVariables(Particles.VisibilityTag=&MakeCustomIntFromBool(Bool=$User.bisHeadHidden, True INT=1, False INT=0)) {isSetParameters=true, module=SetVariables_#6}
  ]
  rendererClasses [&NiagaraSpriteRendererProperties]
  renderers [{rendererClass=&NiagaraSpriteRendererProperties}]
 }
 Flash {
  emitterUpdateScript [
   &EmitterState(
    Inactive Response=::NewEnumerator1
    Life Cycle Mode=::NewEnumerator1
    Loop Behavior=::NewEnumerator1
    Loop Duration Mode=::NewEnumerator1
    UseLoopDelay=false
   )
   &SpawnBurst_Instantaneous(Use Spawn Probability=false, UseLoopCountLimit=false)
  ]
  particleSpawnScript [
   &InitializeParticle(Mesh Scale Mode=::NewEnumerator3, Mesh Uniform Scale=0.92, Position Offset=<0 0 160>, UsePositionOffset=true, Write Material Random=true) {module=InitializeParticle001}
   &InitialMeshOrientation(Rotation=<0.04 0 0.02>, Use Orientation Vector=true, Use Rotation Vector=true) {module=InitialMeshOrientation001}
   &CameraOffset(Camera Offset Amount=0.5) {module=CameraOffset001}
  ]
  particleUpdateScript [
   &ParticleState(Kill Particles When Lifetime Has Elapsed=false, Let Infinitely Lived Particles Die When Emitter Deactivates=true)
   &ScaleMeshSize(Scale Factor=&VectorFromFloat(Value=&Lerp_Float(A=1, B=2, Alpha=$NPC.FortniteNPC.BigHeadMode)))
   &SkeletalMeshLocation(
    MaskSpawnGroup=false
    Mesh Sampling Type=::NewEnumerator5
    Sampled Position Offset=<0 7.4101 -8>
    Skeletal Mesh=$System.Skeletal Mesh Reader
    Socket Sampling Mode=::NewEnumerator5
   )
   &FN_Light_Attributes(
    Color=$User.LightColor
    Create Color Attribute=true
    Create Position Attribute=true
    Create Radius Attribute=true
    Position=&Add_Vector(
     A=Linked() {linkedVariable {name=Particles.Position, type=NiagaraPosition}}
     B=&TransformVector(Vector=&Lerp_Vector(A=<0 -15 0>, B=<0 -30 0>, Alpha=$NPC.FortniteNPC.BigHeadMode), Destination Space=::Simulation)
    )
    Radius=&Lerp_Float(A=50, B=95, Alpha=$NPC.FortniteNPC.BigHeadMode)
   )
   /SetVariables(Particles.VisibilityTag=&MakeCustomIntFromBool(
     Bool=$User.bisHeadHidden
     True INT=-1
     False INT=&SelectIntFromArray(
      Direct Array Index=&MakeIntFromFloat(Float=$User.IsShadow, Conversion Mode=::NewEnumerator1)
      Int Selection Array=DataInterface() {propertyValues {IntData [0,1]}}
     )
    )) {isSetParameters=true, module=SetVariables_#6}
  ]
  rendererClasses [&NiagaraLightRendererProperties, &NiagaraMeshRendererProperties]
  renderers [
   {rendererClass=&NiagaraLightRendererProperties, rendererIndex=0}
   {rendererClass=&NiagaraMeshRendererProperties, rendererIndex=1}
   {rendererClass=&NiagaraMeshRendererProperties, rendererIndex=2}
  ]
 }
}
summary {
 systemName=NS_Head_NinjaWolf
 emitters [
  {emitterName=Flash, rendererClasses [&NiagaraLightRendererProperties, &NiagaraMeshRendererProperties]}
  {emitterName=EyeFlames, simTarget=GPUComputeSim, rendererClasses [&NiagaraSpriteRendererProperties]}
 ]
 userVariables [
  User.bIsFrontEnd {type=NiagaraBool}
  User.bIsFrontEndPreview {type=NiagaraBool}
  User.bisHeadHidden {type=NiagaraBool}
  User.HologramColor {type=LinearColor, description="", defaultValue {struct=LinearColor, value=<rgba 1 1 1 1>}}
  User.IsShadow {type=NiagaraFloat, defaultValue=0, description=""}
  User.LightColor {type=LinearColor, description="", defaultValue {struct=LinearColor, value=<rgba 20 3.0685 8.3336 1>}}
  User.RibbonIDOffset {type=NiagaraInt32}
 ]
}
