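# Physics asset for the left-arm mesh of the Ninja Wolf male character body, covering a dangling arm tassel.
# Capsule bodies sit on the lower arm and five chained dyn tassel bones.
# Constraints link each tassel bone to the previous one and the first to lowerarm_l, with identical moderate swing limits and tight twist limits so the tassel bends but barely twists.
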PhysicsAsset /BRCosmetics/Characters/Player/Male/Medium/Bodies/M_MED_Ninja_Wolf/Meshes/M_MED_Ninja_Wolf_ArmL_Physics
bodies {
 dyn_arm_tassle1_l [{shapeType=Capsule, length=0.505, radius=0.505, shapeName=dyn_arm_tassle1_l_capsule, center=<1.1762 0 0>, rotation=<rot 0 -90 -90>}]
 dyn_arm_tassle2_l [{shapeType=Capsule, length=2, radius=1, shapeName=dyn_arm_tassle2_l_capsule, center=<2.0092 0 0>, rotation=<rot 90 0 0>}]
 dyn_arm_tassle3_l [{shapeType=Capsule, length=2, radius=1, shapeName=dyn_arm_tassle3_l_capsule, center=<2.0092 0 0>, rotation=<rot 90 0 0>}]
 dyn_arm_tassle4_l [{shapeType=Capsule, length=2, radius=1, shapeName=dyn_arm_tassle4_l_capsule, center=<2.0092 0 0>, rotation=<rot 90 0 0>}]
 dyn_arm_tassle5_l [{shapeType=Capsule, length=1, radius=1, shapeName=dyn_arm_tassle5_l_capsule, center=<0 0 -1.5>, rotation=<rot 90 0 0>}]
 lowerarm_l [{shapeType=Capsule, length=1, radius=1, shapeName=lowerarm_l_capsule}]
}
constraints [
 {bone1Name=dyn_arm_tassle2_l, bone2Name=dyn_arm_tassle1_l, swing1LimitDegrees=45, swing2LimitDegrees=45, twistLimitDegrees=5}
 {bone1Name=dyn_arm_tassle3_l, bone2Name=dyn_arm_tassle2_l, swing1LimitDegrees=45, swing2LimitDegrees=45, twistLimitDegrees=5}
 {bone1Name=dyn_arm_tassle4_l, bone2Name=dyn_arm_tassle3_l, swing1LimitDegrees=45, swing2LimitDegrees=45, twistLimitDegrees=5}
 {bone1Name=dyn_arm_tassle5_l, bone2Name=dyn_arm_tassle4_l, swing1LimitDegrees=45, swing2LimitDegrees=45, twistLimitDegrees=5}
 {bone1Name=dyn_arm_tassle1_l, bone2Name=lowerarm_l, swing1LimitDegrees=45, swing2LimitDegrees=45, twistLimitDegrees=5}
]
