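# Physics asset for the right-arm mesh of the male medium Ninja Wolf character body, defining simulation shapes for a dangling arm tassel.
# Single capsule bodies cover the right lower arm and each of five chained dynamic tassel bones.
# Constraints link each tassel bone to the one before it, anchoring the first to the lower arm, with moderate swing and tight twist limits so the chain bends but barely twists.
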PhysicsAsset /BRCosmetics/Characters/Player/Male/Medium/Bodies/M_MED_Ninja_Wolf/Meshes/M_MED_Ninja_Wolf_ArmR_Physics
bodies {
 dyn_arm_tassle1_r [{
   shapeType=Capsule
   length=0.505
   radius=0.505
   shapeName=dyn_arm_tassle1_r_capsule
   center=<1.1762 -0.0675 -0.0511>
   rotation=<rot 0 89.3047 -84.6817>
  }]
 dyn_arm_tassle2_r [{shapeType=Capsule, length=2, radius=1, shapeName=dyn_arm_tassle2_r_capsule, center=<2.0092 -0.0001 0>, rotation=<rot 90 0 0>}]
 dyn_arm_tassle3_r [{shapeType=Capsule, length=2, radius=1, shapeName=dyn_arm_tassle3_r_capsule, center=<2.0092 -0.0001 0>, rotation=<rot 90 0 0>}]
 dyn_arm_tassle4_r [{shapeType=Capsule, length=2, radius=1, shapeName=dyn_arm_tassle4_r_capsule, center=<2.0092 -0.0001 0>, rotation=<rot 90 0 0>}]
 dyn_arm_tassle5_r [{shapeType=Capsule, length=1, radius=1, shapeName=dyn_arm_tassle5_r_capsule, center=<0 0 -1.4543>, rotation=<rot 90 0 0>}]
 lowerarm_r [{shapeType=Capsule, length=1, radius=1, shapeName=lowerarm_r_capsule}]
}
constraints [
 {bone1Name=dyn_arm_tassle2_r, bone2Name=dyn_arm_tassle1_r, swing1LimitDegrees=45, swing2LimitDegrees=45, twistLimitDegrees=5}
 {bone1Name=dyn_arm_tassle3_r, bone2Name=dyn_arm_tassle2_r, swing1LimitDegrees=45, swing2LimitDegrees=45, twistLimitDegrees=5}
 {bone1Name=dyn_arm_tassle4_r, bone2Name=dyn_arm_tassle3_r, swing1LimitDegrees=45, swing2LimitDegrees=45, twistLimitDegrees=5}
 {bone1Name=dyn_arm_tassle5_r, bone2Name=dyn_arm_tassle4_r, swing1LimitDegrees=45, swing2LimitDegrees=45, twistLimitDegrees=5}
 {bone1Name=dyn_arm_tassle1_r, bone2Name=lowerarm_r, swing1LimitDegrees=45, swing2LimitDegrees=45, twistLimitDegrees=5}
]
