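# Physics asset for the left-leg mesh of the Ninja Wolf male medium cosmetic body, defining physics for its dangling leg tassel.
# Small capsule bodies sit on the left thigh and on each of four tassel bones.
# Constraints link the tassel bodies in a serial chain from the thigh down, each with tight, identical swing and twist limits.
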PhysicsAsset /BRCosmetics/Characters/Player/Male/Medium/Bodies/M_MED_Ninja_Wolf/Meshes/M_MED_Ninja_Wolf_LegL_Physics
bodies {
 dyn_leg_tassle1_l [{shapeType=Capsule, length=2, radius=1, shapeName=dyn_leg_tassle1_l_capsule, center=<2.0258 0 0>, rotation=<rot 90 0 0>}]
 dyn_leg_tassle2_l [{shapeType=Capsule, length=2, radius=1, shapeName=dyn_leg_tassle2_l_capsule, center=<2.0258 0 0>, rotation=<rot 90 0 0>}]
 dyn_leg_tassle3_l [{shapeType=Capsule, length=2, radius=1, shapeName=dyn_leg_tassle3_l_capsule, center=<2.0258 0 0>, rotation=<rot 90 0 0>}]
 dyn_leg_tassle4_l [{shapeType=Capsule, length=2, radius=1, shapeName=dyn_leg_tassle4_l_capsule, center=<2.0258 0 0>, rotation=<rot 90 0 0>}]
 thigh_l [{shapeType=Capsule, length=1, radius=1, shapeName=thigh_l_capsule}]
}
constraints [
 {bone1Name=dyn_leg_tassle1_l, bone2Name=thigh_l, swing1LimitDegrees=10, swing2LimitDegrees=10, twistLimitDegrees=5}
 {bone1Name=dyn_leg_tassle2_l, bone2Name=dyn_leg_tassle1_l, swing1LimitDegrees=10, swing2LimitDegrees=10, twistLimitDegrees=5}
 {bone1Name=dyn_leg_tassle3_l, bone2Name=dyn_leg_tassle2_l, swing1LimitDegrees=10, swing2LimitDegrees=10, twistLimitDegrees=5}
 {bone1Name=dyn_leg_tassle4_l, bone2Name=dyn_leg_tassle3_l, swing1LimitDegrees=10, swing2LimitDegrees=10, twistLimitDegrees=5}
]
