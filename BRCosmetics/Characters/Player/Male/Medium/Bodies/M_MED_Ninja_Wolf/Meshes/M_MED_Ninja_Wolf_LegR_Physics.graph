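# Physics asset for the right-leg mesh of the Ninja Wolf male body cosmetic, setting up a dangling tassel on the thigh.
# Four small, identical capsule bodies on the dyn_leg_tassle bones form a chain that hangs from a capsule on thigh_r.
# Each link is constrained to the one above it with tight swing and twist limits, so the tassel bends only slightly per joint.
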PhysicsAsset /BRCosmetics/Characters/Player/Male/Medium/Bodies/M_MED_Ninja_Wolf/Meshes/M_MED_Ninja_Wolf_LegR_Physics
bodies {
 dyn_leg_tassle1_r [{shapeType=Capsule, length=2, radius=1, shapeName=dyn_leg_tassle1_r_capsule, center=<1.8587 0 0>, rotation=<rot 90 0 0>}]
 dyn_leg_tassle2_r [{shapeType=Capsule, length=2, radius=1, shapeName=dyn_leg_tassle2_r_capsule, center=<1.8587 0 0>, rotation=<rot 90 0 0>}]
 dyn_leg_tassle3_r [{shapeType=Capsule, length=2, radius=1, shapeName=dyn_leg_tassle3_r_capsule, center=<1.8587 0 0>, rotation=<rot 90 0 0>}]
 dyn_leg_tassle4_r [{shapeType=Capsule, length=2, radius=1, shapeName=dyn_leg_tassle4_r_capsule, center=<1.8587 0 0>, rotation=<rot 90 0 0>}]
 thigh_r [{shapeType=Capsule, length=1, radius=1, shapeName=thigh_r_capsule}]
}
constraints [
 {bone1Name=dyn_leg_tassle1_r, bone2Name=thigh_r, swing1LimitDegrees=10, swing2LimitDegrees=10, twistLimitDegrees=5}
 {bone1Name=dyn_leg_tassle2_r, bone2Name=dyn_leg_tassle1_r, swing1LimitDegrees=10, swing2LimitDegrees=10, twistLimitDegrees=5}
 {bone1Name=dyn_leg_tassle3_r, bone2Name=dyn_leg_tassle2_r, swing1LimitDegrees=10, swing2LimitDegrees=10, twistLimitDegrees=5}
 {bone1Name=dyn_leg_tassle4_r, bone2Name=dyn_leg_tassle3_r, swing1LimitDegrees=10, swing2LimitDegrees=10, twistLimitDegrees=5}
]
