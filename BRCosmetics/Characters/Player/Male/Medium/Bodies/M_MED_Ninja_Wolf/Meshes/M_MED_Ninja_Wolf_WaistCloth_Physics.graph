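# Physics asset for the waist cloth of the male medium Ninja Wolf cosmetic body, simulating skirt movement with capsule bodies.
# Six bone chains (a–f) of three or four segments each hang from small base capsules constrained to the pelvis.
# Tight swing and twist limits link each segment to its parent, with swing1 narrowing toward the tip in several chains.
# Large thigh capsules, angularly free from the pelvis, give the legs volume alongside the cloth.
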PhysicsAsset /BRCosmetics/Characters/Player/Male/Medium/Bodies/M_MED_Ninja_Wolf/Meshes/M_MED_Ninja_Wolf_WaistCloth_Physics
bodies {
 dyn_skirt01_a [{shapeType=Capsule, length=6.25, radius=3, shapeName=dyn_skirt01_a_capsule, center=<3.7437 -0.0001 0>}]
 dyn_skirt01_c [{shapeType=Capsule, length=7, radius=2.25, shapeName=dyn_skirt01_c_capsule, center=<4.3792 0 0>}]
 dyn_skirt01_e [{shapeType=Capsule, length=6.25, radius=3, shapeName=dyn_skirt01_e_capsule, center=<3.1712 0.0001 0>}]
 dyn_skirt02_a [{shapeType=Capsule, length=6.25, radius=3, shapeName=dyn_skirt02_a_capsule, center=<3.7437 -0.0001 0>}]
 dyn_skirt02_c [{shapeType=Capsule, length=6.25, radius=3, shapeName=dyn_skirt02_c_capsule, center=<4.3792 0 0>}]
 dyn_skirt02_e [{shapeType=Capsule, length=11.5, radius=2.5, shapeName=dyn_skirt2_f_capsule, center=<3.1712 0.0001 0>}]
 dyn_skirt03_a [{shapeType=Capsule, length=6.25, radius=3, shapeName=dyn_skirt03_a_capsule, center=<3.7437 -0.0001 0>}]
 dyn_skirt03_c [{shapeType=Capsule, length=6.25, radius=3, shapeName=dyn_skirt03_c_capsule, center=<4.3792 0 0>}]
 dyn_skirt03_e [{shapeType=Capsule, length=11.5, radius=2.5, shapeName=dyn_skirt2_f_capsule, center=<3.1712 0.0001 0>}]
 dyn_skirt04_c [{shapeType=Capsule, length=6.25, radius=3, shapeName=dyn_skirt04_c_capsule, center=<4.3792 0 0>}]
 dyn_skirt04_e [{shapeType=Capsule, length=11.5, radius=2.5, shapeName=dyn_skirt2_f_capsule, center=<3.1712 0.0001 0>}]
 dyn_skirt1_b [{shapeType=Capsule, length=6.25, radius=3, shapeName=dyn_skirt1_b_capsule, center=<3.5659 0 0>}]
 dyn_skirt1_d [{shapeType=Capsule, length=7, radius=2.25, shapeName=dyn_skirt1_d_capsule, center=<4.727 0 0>}]
 dyn_skirt1_f [{shapeType=Capsule, length=5.5, radius=3.75, shapeName=dyn_skirt2_f_capsule, center=<3.9618 0.9481 0.0001>}]
 dyn_skirt2_b [{shapeType=Capsule, length=6.25, radius=3, shapeName=dyn_skirt2_b_capsule, center=<3.5659 0 0>}]
 dyn_skirt2_d [{shapeType=Capsule, length=7, radius=2.25, shapeName=dyn_skirt2_d_capsule, center=<3.2092 0.0001 0>}]
 dyn_skirt2_f [{shapeType=Capsule, length=6.5, radius=2.75, shapeName=dyn_skirt2_f_capsule, center=<3.9618 0.9481 0.0001>}]
 dyn_skirt3_b [{shapeType=Capsule, length=6.25, radius=3, shapeName=dyn_skirt3_b_capsule, center=<3.5659 0 0>}]
 dyn_skirt3_d [{shapeType=Capsule, length=7, radius=2.25, shapeName=dyn_skirt3_d_capsule, center=<3.2092 0.0001 0>}]
 dyn_skirt3_f [{shapeType=Capsule, length=6.75, radius=2.5, shapeName=dyn_skirt2_f_capsule, center=<3.9618 0.9481 0.0001>}]
 dyn_skirt4_d [{shapeType=Capsule, length=6.25, radius=3, shapeName=dyn_skirt4_d_capsule, center=<3.2092 0.0001 0>}]
 dyn_skirt_a_base [{shapeType=Capsule, length=1, radius=1, shapeName=dyn_skirt_a_base_capsule}]
 dyn_skirt_b_base [{shapeType=Capsule, length=1, radius=1, shapeName=dyn_skirt_b_base_capsule}]
 dyn_skirt_c_base [{shapeType=Capsule, length=1, radius=1, shapeName=dyn_skirt_c_base_capsule}]
 dyn_skirt_d_base [{shapeType=Capsule, length=1, radius=1, shapeName=dyn_skirt_d_base_capsule}]
 dyn_skirt_e_base [{shapeType=Capsule, length=1, radius=1, shapeName=dyn_skirt_e_base_capsule}]
 dyn_skirt_f_base [{shapeType=Capsule, length=1, radius=1, shapeName=dyn_skirt_f_base_capsule}]
 pelvis [{shapeType=Capsule, length=3, radius=3, shapeName=pelvis_capsule}]
 thigh_l [{shapeType=Capsule, length=31.75, radius=10.5, shapeName=thigh_r_capsule, center=<-21.9845 0.0002 -1.3715>, rotation=<rot 90 0 0>}]
 thigh_r [{shapeType=Capsule, length=31.75, radius=10.5, shapeName=thigh_r_capsule, center=<21.9845 0.0002 -1.3715>, rotation=<rot 90 0 0>}]
}
constraints [
 {bone1Name=dyn_skirt01_a, bone2Name=dyn_skirt_a_base, swing1LimitDegrees=15, swing2LimitDegrees=5, twistLimitDegrees=5}
 {bone1Name=dyn_skirt02_a, bone2Name=dyn_skirt01_a, swing1LimitDegrees=15, swing2LimitDegrees=5, twistLimitDegrees=5}
 {bone1Name=dyn_skirt03_a, bone2Name=dyn_skirt02_a, swing1LimitDegrees=15, swing2LimitDegrees=5, twistLimitDegrees=5}
 {bone1Name=dyn_skirt1_b, bone2Name=dyn_skirt_b_base, swing1LimitDegrees=15, swing2LimitDegrees=5, twistLimitDegrees=5}
 {bone1Name=dyn_skirt2_b, bone2Name=dyn_skirt1_b, swing1LimitDegrees=10, swing2LimitDegrees=5, twistLimitDegrees=5}
 {bone1Name=dyn_skirt3_b, bone2Name=dyn_skirt2_b, swing1LimitDegrees=5, swing2LimitDegrees=5, twistLimitDegrees=5}
 {bone1Name=dyn_skirt01_c, bone2Name=dyn_skirt_c_base, swing1LimitDegrees=15, swing2LimitDegrees=5, twistLimitDegrees=5}
 {bone1Name=dyn_skirt02_c, bone2Name=dyn_skirt01_c, swing1LimitDegrees=15, swing2LimitDegrees=5, twistLimitDegrees=5}
 {bone1Name=dyn_skirt03_c, bone2Name=dyn_skirt02_c, swing1LimitDegrees=10, swing2LimitDegrees=5, twistLimitDegrees=5}
 {bone1Name=dyn_skirt04_c, bone2Name=dyn_skirt03_c, swing1LimitDegrees=5, swing2LimitDegrees=5, twistLimitDegrees=5}
 {bone1Name=dyn_skirt1_d, bone2Name=dyn_skirt_d_base, swing1LimitDegrees=15, swing2LimitDegrees=5, twistLimitDegrees=5}
 {bone1Name=dyn_skirt2_d, bone2Name=dyn_skirt1_d, swing1LimitDegrees=15, swing2LimitDegrees=5, twistLimitDegrees=5}
 {bone1Name=dyn_skirt3_d, bone2Name=dyn_skirt2_d, swing1LimitDegrees=10, swing2LimitDegrees=5, twistLimitDegrees=5}
 {bone1Name=dyn_skirt4_d, bone2Name=dyn_skirt3_d, swing1LimitDegrees=10, swing2LimitDegrees=5, twistLimitDegrees=5}
 {bone1Name=dyn_skirt01_e, bone2Name=dyn_skirt_e_base, swing1LimitDegrees=15, swing2LimitDegrees=5, twistLimitDegrees=5}
 {bone1Name=dyn_skirt_a_base, bone2Name=pelvis, swing1LimitDegrees=15, swing2LimitDegrees=15, twistLimitDegrees=5}
 {bone1Name=dyn_skirt_b_base, bone2Name=pelvis, swing1LimitDegrees=15, swing2LimitDegrees=15, twistLimitDegrees=5}
 {bone1Name=dyn_skirt_c_base, bone2Name=pelvis, swing1LimitDegrees=15, swing2LimitDegrees=15, twistLimitDegrees=5}
 {bone1Name=dyn_skirt_d_base, bone2Name=pelvis, swing1LimitDegrees=15, swing2LimitDegrees=15, twistLimitDegrees=5}
 {bone1Name=dyn_skirt_e_base, bone2Name=pelvis, swing1LimitDegrees=15, swing2LimitDegrees=15, twistLimitDegrees=5}
 {bone1Name=dyn_skirt_f_base, bone2Name=pelvis, swing1LimitDegrees=15, swing2LimitDegrees=15, twistLimitDegrees=5}
 {
  bone1Name=thigh_r
  bone2Name=pelvis
  swing1LimitDegrees=45
  swing1Motion=Free
  swing2LimitDegrees=45
  swing2Motion=Free
  twistLimitDegrees=45
  twistMotion=Free
 }
 {
  bone1Name=thigh_l
  bone2Name=pelvis
  swing1LimitDegrees=45
  swing1Motion=Free
  swing2LimitDegrees=45
  swing2Motion=Free
  twistLimitDegrees=45
  twistMotion=Free
 }
 {bone1Name=dyn_skirt02_e, bone2Name=dyn_skirt01_e, swing1LimitDegrees=15, swing2LimitDegrees=5, twistLimitDegrees=5}
 {bone1Name=dyn_skirt03_e, bone2Name=dyn_skirt02_e, swing1LimitDegrees=10, swing2LimitDegrees=10, twistLimitDegrees=5}
 {bone1Name=dyn_skirt04_e, bone2Name=dyn_skirt03_e, swing1LimitDegrees=10, swing2LimitDegrees=10, twistLimitDegrees=5}
 {bone1Name=dyn_skirt1_f, bone2Name=dyn_skirt_f_base, swing1LimitDegrees=15, swing2LimitDegrees=5, twistLimitDegrees=5}
 {bone1Name=dyn_skirt2_f, bone2Name=dyn_skirt1_f, swing1LimitDegrees=15, swing2LimitDegrees=5, twistLimitDegrees=5}
 {bone1Name=dyn_skirt3_f, bone2Name=dyn_skirt2_f, swing1LimitDegrees=15, swing2LimitDegrees=5, twistLimitDegrees=5}
]
