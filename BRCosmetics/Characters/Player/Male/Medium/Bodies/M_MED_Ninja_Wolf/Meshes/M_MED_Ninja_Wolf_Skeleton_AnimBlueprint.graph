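# Post-pose Anim Blueprint for the Ninja Wolf character body, adjusting its skirt, pouch, pauldron and sleeve.
# It copies the attached parent's pose, then chains additive bone offsets weighted by thigh and arm alphas, crouch, jump, skydive, pickaxe and glider variables.
# RigidBody nodes simulate waist cloth, arms and legs between those corrective passes.
# A separate chain repeats the copy and simulation, then constrains skirt bones to pelvis and thigh twists.
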AnimBlueprint /BRCosmetics/Characters/Player/Male/Medium/Bodies/M_MED_Ninja_Wolf/Meshes/M_MED_Ninja_Wolf_Skeleton_AnimBlueprint
target_skeleton=&/M_MED_Ninja_Wolf_Skeleton
graphs [
 AnimGraph {nodes [
   n1 = Get thigh_r_rot_alpha() -> (thigh_r_rot_alpha:double)
   n2 = "Transform (Modify) Bone - Bone: dyn_skirt_a_base"(
    ComponentPose=Local To Component(LocalPose=Copy Pose From Mesh() {bUseAttachedParent=true})
    Alpha=@n1
    Translation="-2.367,-0.636,-0.000"
    Rotation="0.000,-45,0.000"
    Scale="1,1,1"
   ) {
    rotationMode=BMM_Additive
    rotationSpace=BCS_BoneSpace
    translationMode=BMM_Additive
    translationSpace=BCS_BoneSpace
    alphaBoolBlend {blendOption=Linear}
    boneToModify {boneName=dyn_skirt_a_base}
    rotation {yaw=-45}
    translation=<-2.367 -0.636>
   }
   n3 = "Transform (Modify) Bone - Bone: dyn_skirt_b_base"(ComponentPose=@n2, Alpha=@n1, Translation="-8.442,-0.000,0.000", Rotation="0.000,-60,0.000", Scale="1,1,1") {
    rotationMode=BMM_Additive
    rotationSpace=BCS_BoneSpace
    translationMode=BMM_Additive
    translationSpace=BCS_BoneSpace
    alphaBoolBlend {blendOption=Linear}
    boneToModify {boneName=dyn_skirt_b_base}
    rotation {yaw=-60}
    translation {x=-8.442}
   }
   n4 = "Transform (Modify) Bone - Bone: dyn_skirt_c_base"(ComponentPose=@n3, Alpha=@n1, Translation="-8.23,0.000,0.000", Rotation="0,-60,0", Scale="1,1,1") {
    rotationMode=BMM_Additive
    rotationSpace=BCS_BoneSpace
    translationMode=BMM_Additive
    translationSpace=BCS_BoneSpace
    alphaBoolBlend {blendOption=Linear}
    boneToModify {boneName=dyn_skirt_c_base}
    rotation {yaw=-60}
    translation {x=-8.23}
   }
   n5 = "Transform (Modify) Bone - Bone: dyn_skirt_d_base"(ComponentPose=@n4, Alpha=@n1, Translation="-9.444,-2.148,-0.000", Rotation="15,-40,0.000", Scale="1,1,1") {
    rotationMode=BMM_Additive
    rotationSpace=BCS_BoneSpace
    translationMode=BMM_Additive
    translationSpace=BCS_BoneSpace
    alphaBoolBlend {blendOption=Linear}
    boneToModify {boneName=dyn_skirt_d_base}
    rotation {pitch=15, yaw=-40}
    translation=<-9.444 -2.148>
   }
   n6 = "Transform (Modify) Bone - Bone: dyn_skirt_e_base"(ComponentPose=@n5, Alpha=@n1, Translation="0,0,0", Rotation="-15,-15,0.000", Scale="1,1,1") {
    rotationMode=BMM_Additive
    rotationSpace=BCS_BoneSpace
    translationMode=BMM_Additive
    translationSpace=BCS_BoneSpace
    alphaBoolBlend {blendOption=Linear}
    boneToModify {boneName=dyn_skirt_e_base}
    rotation {pitch=-15, yaw=-15}
   }
   n7 = "Transform (Modify) Bone - Bone: dyn_pouch_base"(ComponentPose=@n6, Alpha=@n1, Translation="-10.483,0.150,-2.033", Rotation="-37.893,57.025,1.630", Scale="1,1,1") {
    rotationMode=BMM_Additive
    rotationSpace=BCS_BoneSpace
    translationMode=BMM_Additive
    translationSpace=BCS_BoneSpace
    alphaBoolBlend {blendOption=Linear}
    boneToModify {boneName=dyn_pouch_base}
    rotation=<rot -37.893 57.0249 1.63>
    translation=<-10.483 0.15 -2.033>
   }
   n8 = Get thigh_r_pitch_alpha() -> (thigh_r_pitch_alpha:double)
   n9 = "Transform (Modify) Bone - Bone: dyn_skirt01_c"(ComponentPose=@n7, Alpha=@n8, Translation="0,0,0", Rotation="0,-15,0", Scale="1,1,1") {
    rotationMode=BMM_Additive
    rotationSpace=BCS_BoneSpace
    translationMode=BMM_Additive
    translationSpace=BCS_BoneSpace
    alphaBoolBlend {blendOption=Linear}
    boneToModify {boneName=dyn_skirt01_c}
    rotation {yaw=-15}
   }
   n10 = "Transform (Modify) Bone - Bone: dyn_skirt_d_base"(ComponentPose=@n9, Alpha=@n8, Translation="0,0,0", Rotation="0,-45,0", Scale="1,1,1") {
    rotationMode=BMM_Additive
    rotationSpace=BCS_BoneSpace
    translationMode=BMM_Additive
    translationSpace=BCS_BoneSpace
    alphaBoolBlend {blendOption=Linear}
    boneToModify {boneName=dyn_skirt_d_base}
    rotation {yaw=-45}
   }
   n11 = "Transform (Modify) Bone - Bone: dyn_skirt_e_base"(ComponentPose=@n10, Alpha=@n8, Translation="-4,0,0", Rotation="0,-25,0", Scale="1,1,1") {
    rotationMode=BMM_Additive
    rotationSpace=BCS_BoneSpace
    translationMode=BMM_Additive
    translationSpace=BCS_BoneSpace
    alphaBoolBlend {blendOption=Linear}
    boneToModify {boneName=dyn_skirt_e_base}
    rotation {yaw=-25}
    translation {x=-4}
   }
   n12 = Get bIsCrouching()
   n13 = "Transform (Modify) Bone - Bone: dyn_skirt_a_base"!(ComponentPose=@n11, bAlphaBoolEnabled:bool=@n12, Translation="-5,0,0.000", Rotation="-1.017,-68.753,36.672", Scale="1,1,1") -> (Pose) {
    alphaInputType=Bool
    rotationMode=BMM_Additive
    rotationSpace=BCS_BoneSpace
    translationMode=BMM_Additive
    translationSpace=BCS_BoneSpace
    alphaBoolBlend {blendOption=Linear}
    boneToModify {boneName=dyn_skirt_a_base}
    rotation=<rot -1.017 -68.753 36.6716>
    translation {x=-5}
   }
   n14 = "Transform (Modify) Bone - Bone: dyn_skirt03_a"!(ComponentPose=@n13, bAlphaBoolEnabled:bool=@n12, Translation="0,0,0", Rotation="20.019,-3.716,-2.47", Scale="1,1,1") -> (Pose) {
    alphaInputType=Bool
    rotationMode=BMM_Additive
    rotationSpace=BCS_BoneSpace
    translationMode=BMM_Additive
    translationSpace=BCS_BoneSpace
    alphaBoolBlend {blendOption=Linear}
    boneToModify {boneName=dyn_skirt03_a}
    rotation=<rot 20.019 -3.716 -2.47>
   }
   n15 = "Transform (Modify) Bone - Bone: dyn_skirt_belt_helper_01"!(ComponentPose=@n14, bAlphaBoolEnabled:bool=@n12, Translation="-2.401,-2.902,0.886", Rotation="0.000,100,0.000", Scale="1,1,1") -> (Pose) {
    alphaInputType=Bool
    rotationMode=BMM_Additive
    rotationSpace=BCS_BoneSpace
    translationMode=BMM_Additive
    translationSpace=BCS_BoneSpace
    alphaBoolBlend {blendOption=Linear}
    boneToModify {boneName=dyn_skirt_belt_helper_01}
    rotation {yaw=100}
    translation=<-2.401 -2.902 0.886>
   }
   n16 = "Transform (Modify) Bone - Bone: dyn_skirt_a_base"(ComponentPose=@n15, bAlphaBoolEnabled=@n12, Translation="0,0,0", Rotation="0,6.215,0.000", Scale="1,1,1") {
    alphaInputType=Bool
    rotationMode=BMM_Additive
    rotationSpace=BCS_BoneSpace
    translationMode=BMM_Additive
    translationSpace=BCS_BoneSpace
    alphaBoolBlend {blendOption=Linear}
    boneToModify {boneName=dyn_skirt_a_base}
    rotation {yaw=6.215}
   }
   n17 = "Transform (Modify) Bone - Bone: dyn_pouch_base"!(ComponentPose=@n16, bAlphaBoolEnabled:bool=@n12, Translation="-2,0,-2", Rotation="0.000,20,0", Scale="1,1,1") -> (Pose) {
    alphaInputType=Bool
    rotationMode=BMM_Additive
    rotationSpace=BCS_BoneSpace
    scaleMode=BMM_Replace
    translationMode=BMM_Additive
    translationSpace=BCS_BoneSpace
    alphaBoolBlend {blendOption=Linear}
    boneToModify {boneName=dyn_pouch_base}
    rotation {yaw=20}
    translation {x=-2, z=-2}
   }
   n18 = "Transform (Modify) Bone - Bone: dyn_skirt_b_base"!(ComponentPose=@n17, bAlphaBoolEnabled:bool=@n12, Translation="-2,0,0", Rotation="0,-40,0", Scale="1,1,1") -> (Pose) {
    alphaInputType=Bool
    rotationMode=BMM_Additive
    rotationSpace=BCS_BoneSpace
    translationMode=BMM_Additive
    translationSpace=BCS_BoneSpace
    alphaBoolBlend {blendOption=Linear}
    boneToModify {boneName=dyn_skirt_b_base}
    rotation {yaw=-40}
    translation {x=-2}
   }
   n19 = "Transform (Modify) Bone - Bone: dyn_skirt_c_base"!(ComponentPose=@n18, bAlphaBoolEnabled:bool=@n12, Translation="0,0,0", Rotation="0,-20,0", Scale="1,1,1") -> (Pose) {
    alphaInputType=Bool
    rotationMode=BMM_Additive
    rotationSpace=BCS_BoneSpace
    translationMode=BMM_Additive
    translationSpace=BCS_BoneSpace
    alphaBoolBlend {blendOption=Linear}
    boneToModify {boneName=dyn_skirt_c_base}
    rotation {yaw=-20}
   }
   n20 = "Transform (Modify) Bone - Bone: dyn_skirt_d_base"!(ComponentPose=@n19, bAlphaBoolEnabled:bool=@n12, Translation="0,0,0", Rotation="0,0,0", Scale="1,1,1") -> (Pose) {
    alphaInputType=Bool
    rotationMode=BMM_Additive
    rotationSpace=BCS_BoneSpace
    translationMode=BMM_Additive
    translationSpace=BCS_BoneSpace
    alphaBoolBlend {blendOption=Linear}
    boneToModify {boneName=dyn_skirt_d_base}
   }
   n21 = Get thigh_r_crouch_rot_alpha() -> (thigh_r_crouch_rot_alpha:double)
   n22 = "Transform (Modify) Bone - Bone: dyn_skirt_a_base"!(ComponentPose=@n20, Alpha:float=@n21, Translation="0,0,0", Rotation="0,0,0", Scale="1,1,1") -> (Pose) {
    rotationMode=BMM_Additive
    rotationSpace=BCS_BoneSpace
    translationMode=BMM_Additive
    translationSpace=BCS_BoneSpace
    alphaBoolBlend {blendOption=Linear}
    boneToModify {boneName=dyn_skirt_a_base}
   }
   n23 = "Transform (Modify) Bone - Bone: dyn_skirt_b_base"!(ComponentPose=@n22, Alpha:float=@n21, Translation="-8.442,-0.000,0.000", Rotation="0.000,-50,0.000", Scale="1,1,1") -> (Pose) {
    rotationMode=BMM_Additive
    rotationSpace=BCS_BoneSpace
    translationMode=BMM_Additive
    translationSpace=BCS_BoneSpace
    alphaBoolBlend {blendOption=Linear}
    boneToModify {boneName=dyn_skirt_b_base}
    rotation {yaw=-50}
    translation {x=-8.442}
   }
   n24 = "Transform (Modify) Bone - Bone: dyn_skirt_c_base"!(ComponentPose=@n23, Alpha:float=@n21, Translation="-11.870,1.247,-0.000", Rotation="0,-70,0", Scale="1,1,1") -> (Pose) {
    rotationMode=BMM_Additive
    rotationSpace=BCS_BoneSpace
    translationMode=BMM_Additive
    translationSpace=BCS_BoneSpace
    alphaBoolBlend {blendOption=Linear}
    boneToModify {boneName=dyn_skirt_c_base}
    rotation {yaw=-70}
    translation=<-11.87 1.247>
   }
   n25 = "Transform (Modify) Bone - Bone: dyn_skirt_d_base"!(ComponentPose=@n24, Alpha:float=@n21, Translation="-10,-2,-0.000", Rotation="-15,10,15", Scale="1,1,1") -> (Pose) {
    rotationMode=BMM_Additive
    rotationSpace=BCS_BoneSpace
    translationMode=BMM_Additive
    translationSpace=BCS_BoneSpace
    alphaBoolBlend {blendOption=Linear}
    boneToModify {boneName=dyn_skirt_d_base}
    rotation=<rot -15 10 15>
    translation=<-10 -2>
   }
   n26 = "Transform (Modify) Bone - Bone: dyn_skirt02_e"(ComponentPose=@n25, Alpha=@n21, Translation="0,0,0", Rotation="0,20,0.000", Scale="1,1,1") {
    rotationMode=BMM_Additive
    rotationSpace=BCS_BoneSpace
    translationMode=BMM_Additive
    translationSpace=BCS_BoneSpace
    alphaBoolBlend {blendOption=Linear}
    boneToModify {boneName=dyn_skirt02_e}
    rotation {yaw=20}
   }
   n27 = "Transform (Modify) Bone - Bone: dyn_skirt2_f"(ComponentPose=@n26, Alpha=@n21, Translation="0,0,0", Rotation="0,20,0.000", Scale="1,1,1") {
    rotationMode=BMM_Additive
    rotationSpace=BCS_BoneSpace
    translationMode=BMM_Additive
    translationSpace=BCS_BoneSpace
    alphaBoolBlend {blendOption=Linear}
    boneToModify {boneName=dyn_skirt2_f}
    rotation {yaw=20}
   }
   n28 = "Transform (Modify) Bone - Bone: dyn_pouch_base"!(ComponentPose=@n27, Alpha:float=@n21, Translation="-11.706,3.463,-4.076", Rotation="-10.262,72.436,-4.687", Scale="1,1,1") -> (Pose) {
    rotationMode=BMM_Additive
    rotationSpace=BCS_BoneSpace
    scaleMode=BMM_Replace
    translationMode=BMM_Additive
    translationSpace=BCS_BoneSpace
    alphaBoolBlend {blendOption=Linear}
    boneToModify {boneName=dyn_pouch_base}
    rotation=<rot -10.262 72.4358 -4.687>
    translation=<-11.706 3.463 -4.076>
   }
   n29 = "Transform (Modify) Bone - Bone: dyn_skirt_c_base"(ComponentPose=@n28, Alpha=@n21, Translation="0,0,0", Rotation="0,-10,0", Scale="1,1,1") {
    rotationMode=BMM_Additive
    rotationSpace=BCS_BoneSpace
    translationMode=BMM_Additive
    translationSpace=BCS_BoneSpace
    alphaBoolBlend {blendOption=Linear}
    boneToModify {boneName=dyn_skirt_c_base}
    rotation {yaw=-10}
   }
   n30 = Get thigh_r_crouch_pitch_alpha() -> (thigh_r_crouch_pitch_alpha:double)
   n31 = "Transform (Modify) Bone - Bone: dyn_skirt_d_base"(ComponentPose=@n29, Alpha=@n30, Translation="0,0,0", Rotation="0,-70,0", Scale="1,1,1") {
    rotationMode=BMM_Additive
    rotationSpace=BCS_BoneSpace
    translationMode=BMM_Additive
    translationSpace=BCS_BoneSpace
    alphaBoolBlend {blendOption=Linear}
    boneToModify {boneName=dyn_skirt_d_base}
    rotation {yaw=-70}
   }
   n32 = "Transform (Modify) Bone - Bone: dyn_pouch_base"(ComponentPose=@n31, Alpha=@n30, Translation="0.75,1.6,-1.3", Rotation="0.000,-10,0.000", Scale="1,1,1") {
    rotationMode=BMM_Additive
    rotationSpace=BCS_BoneSpace
    scaleMode=BMM_Replace
    translationMode=BMM_Additive
    translationSpace=BCS_BoneSpace
    alphaBoolBlend {blendOption=Linear}
    boneToModify {boneName=dyn_pouch_base}
    rotation {yaw=-10}
    translation=<0.75 1.6 -1.3>
   }
   n33 = "Transform (Modify) Bone - Bone: dyn_skirt03_c"(
    ComponentPose=@n32
    Alpha=Get thigh_r_crouch_rot_roll_alpha() -> (thigh_r_crouch_rot_roll_alpha:double)
    Translation="0,0,0"
    Rotation="0,10,0"
    Scale="1,1,1"
   ) {
    rotationMode=BMM_Additive
    rotationSpace=BCS_BoneSpace
    translationMode=BMM_Additive
    translationSpace=BCS_BoneSpace
    alphaBoolBlend {blendOption=Linear}
    boneToModify {boneName=dyn_skirt03_c}
    rotation {yaw=10}
   }
   n34 = Get thigh_r_crouch_rot_roll_alpha()
   n35 = "Transform (Modify) Bone - Bone: dyn_skirt01_e"(ComponentPose=@n33, Alpha=@n34, Translation="0,0,0", Rotation="0,20,0", Scale="1,1,1") {
    rotationMode=BMM_Additive
    rotationSpace=BCS_BoneSpace
    translationMode=BMM_Additive
    translationSpace=BCS_BoneSpace
    alphaBoolBlend {blendOption=Linear}
    boneToModify {boneName=dyn_skirt01_e}
    rotation {yaw=20}
   }
   n36 = "Transform (Modify) Bone - Bone: dyn_skirt1_f"(ComponentPose=@n35, Alpha=@n34, Translation="0,0,0", Rotation="0,20,0", Scale="1,1,1") {
    rotationMode=BMM_Additive
    rotationSpace=BCS_BoneSpace
    translationMode=BMM_Additive
    translationSpace=BCS_BoneSpace
    alphaBoolBlend {blendOption=Linear}
    boneToModify {boneName=dyn_skirt1_f}
    rotation {yaw=20}
   }
   n37 = Get jump_alpha() -> (jump_alpha:double)
   n38 = "Transform (Modify) Bone - Bone: dyn_skirt_a_base"(ComponentPose=@n36, Alpha=@n37, Translation="0,0,0", Rotation="0.000,-15,0.000", Scale="1,1,1") {
    rotationMode=BMM_Additive
    rotationSpace=BCS_BoneSpace
    translationMode=BMM_Additive
    translationSpace=BCS_BoneSpace
    alphaBoolBlend {blendOption=Linear}
    boneToModify {boneName=dyn_skirt_a_base}
    rotation {yaw=-15}
   }
   n39 = "Transform (Modify) Bone - Bone: dyn_skirt_b_base"(ComponentPose=@n38, Alpha=@n37, Translation="0,0,0", Rotation="0.000,-15,0.000", Scale="1,1,1") {
    rotationMode=BMM_Additive
    rotationSpace=BCS_BoneSpace
    translationMode=BMM_Additive
    translationSpace=BCS_BoneSpace
    alphaBoolBlend {blendOption=Linear}
    boneToModify {boneName=dyn_skirt_b_base}
    rotation {yaw=-15}
   }
   n40 = "Transform (Modify) Bone - Bone: dyn_skirt_c_base"(ComponentPose=@n39, Alpha=@n37, Translation="0,0,0", Rotation="0,-20,0", Scale="1,1,1") {
    rotationMode=BMM_Additive
    rotationSpace=BCS_BoneSpace
    translationMode=BMM_Additive
    translationSpace=BCS_BoneSpace
    alphaBoolBlend {blendOption=Linear}
    boneToModify {boneName=dyn_skirt_c_base}
    rotation {yaw=-20}
   }
   n41 = "Transform (Modify) Bone - Bone: dyn_skirt_d_base"(ComponentPose=@n40, Alpha=@n37, Translation="0,0,0", Rotation="0.000,-35,0.000", Scale="1,1,1") {
    rotationMode=BMM_Additive
    rotationSpace=BCS_BoneSpace
    translationMode=BMM_Additive
    translationSpace=BCS_BoneSpace
    alphaBoolBlend {blendOption=Linear}
    boneToModify {boneName=dyn_skirt_d_base}
    rotation {yaw=-35}
   }
   n42 = "Transform (Modify) Bone - Bone: dyn_skirt_e_base"(ComponentPose=@n41, Alpha=@n37, Translation="0,0,0", Rotation="0.000,15,0.000", Scale="1,1,1") {
    rotationMode=BMM_Additive
    rotationSpace=BCS_BoneSpace
    translationMode=BMM_Additive
    translationSpace=BCS_BoneSpace
    alphaBoolBlend {blendOption=Linear}
    boneToModify {boneName=dyn_skirt_e_base}
    rotation {yaw=15}
   }
   n43 = "Transform (Modify) Bone - Bone: dyn_skirt_f_base"(ComponentPose=@n42, Alpha=@n37, Translation="0,0,0", Rotation="0.000,-15,0.000", Scale="1,1,1") {
    rotationMode=BMM_Additive
    rotationSpace=BCS_BoneSpace
    translationMode=BMM_Additive
    translationSpace=BCS_BoneSpace
    alphaBoolBlend {blendOption=Linear}
    boneToModify {boneName=dyn_skirt_f_base}
    rotation {yaw=-15}
   }
   n44 = "Transform (Modify) Bone - Bone: dyn_pouch_base"(ComponentPose=@n43, Alpha=@n37, Translation="0,0,0", Rotation="-10,15,0", Scale="1,1,1") {
    rotationMode=BMM_Additive
    rotationSpace=BCS_BoneSpace
    translationMode=BMM_Additive
    translationSpace=BCS_BoneSpace
    alphaBoolBlend {blendOption=Linear}
    boneToModify {boneName=dyn_pouch_base}
    rotation {pitch=-10, yaw=15}
   }
   n45 = RigidBody(ComponentPose=@n44, Alpha=Get cloth_alpha() -> (cloth_alpha:double)) {
    bOverrideWorldGravity=true
    overridePhysicsAsset=&/M_MED_Ninja_Wolf_WaistCloth_Physics
    simulationSpace=BaseBoneSpace
    alphaBoolBlend {blendOption=Linear}
    baseBoneRef {boneName=pelvis}
    componentAppliedLinearAccClamp=<2000 2000 2000>
    componentLinearAccScale=<0.3 0.3 0.35>
    componentLinearVelScale=<0.3 0.3 0.35>
    overrideWorldGravity {z=-980}
   }
   n46 = RigidBody(ComponentPose=@n45) {
    bOverrideWorldGravity=true
    overridePhysicsAsset=&/M_MED_Ninja_Wolf_ArmL_Physics
    simulationSpace=BaseBoneSpace
    alphaBoolBlend {blendOption=Linear}
    baseBoneRef {boneName=lowerarm_l}
    componentAppliedLinearAccClamp=<1500 1500 1500>
    componentLinearAccScale=<0.25 0.25 0.3>
    componentLinearVelScale=<0.25 0.25 0.3>
    overrideWorldGravity {z=-980}
   }
   n47 = RigidBody(ComponentPose=@n46) {
    bOverrideWorldGravity=true
    overridePhysicsAsset=&/M_MED_Ninja_Wolf_ArmR_Physics
    simulationSpace=BaseBoneSpace
    alphaBoolBlend {blendOption=Linear}
    baseBoneRef {boneName=lowerarm_r}
    componentAppliedLinearAccClamp=<1500 1500 1500>
    componentLinearAccScale=<0.25 0.25 0.3>
    componentLinearVelScale=<0.25 0.25 0.3>
    overrideWorldGravity {z=-980}
   }
   n48 = RigidBody(ComponentPose=@n47) {
    bOverrideWorldGravity=true
    overridePhysicsAsset=&/M_MED_Ninja_Wolf_LegL_Physics
    simulationSpace=BaseBoneSpace
    alphaBoolBlend {blendOption=Linear}
    baseBoneRef {boneName=thigh_l}
    componentAppliedLinearAccClamp=<1500 1500 1500>
    componentLinearAccScale=<0.25 0.25 0.3>
    componentLinearVelScale=<0.25 0.25 0.3>
    overrideWorldGravity {z=-980}
   }
   n49 = RigidBody(ComponentPose=@n48) {
    bOverrideWorldGravity=true
    overridePhysicsAsset=&/M_MED_Ninja_Wolf_LegR_Physics
    simulationSpace=BaseBoneSpace
    alphaBoolBlend {blendOption=Linear}
    baseBoneRef {boneName=thigh_r}
    componentAppliedLinearAccClamp=<1500 1500 1500>
    componentLinearAccScale=<0.25 0.25 0.3>
    componentLinearVelScale=<0.25 0.25 0.3>
    overrideWorldGravity {z=-980}
   }
   n50 = "Transform (Modify) Bone - Bone: dyn_skirt_a_base"(ComponentPose=@n49, Alpha=1, Translation="0,0,0", Rotation="0,-10,0", Scale="1,1,1") {
    rotationMode=BMM_Additive
    rotationSpace=BCS_BoneSpace
    alphaBoolBlend {blendOption=Linear}
    boneToModify {boneName=dyn_skirt_a_base}
    rotation {yaw=-10}
   }
   n51 = "Transform (Modify) Bone - Bone: dyn_skirt_c_base"(ComponentPose=@n50, Alpha=1, Translation="0,0,0", Rotation="0,-10,0", Scale="1,1,1") {
    rotationMode=BMM_Additive
    rotationSpace=BCS_BoneSpace
    alphaBoolBlend {blendOption=Linear}
    boneToModify {boneName=dyn_skirt_c_base}
    rotation {yaw=-10}
   }
   n52 = "Transform (Modify) Bone - Bone: dyn_skirt_d_base"(ComponentPose=@n51, Alpha=1, Translation="0,0,0", Rotation="0,-10,0", Scale="1,1,1") {
    rotationMode=BMM_Additive
    rotationSpace=BCS_BoneSpace
    alphaBoolBlend {blendOption=Linear}
    boneToModify {boneName=dyn_skirt_d_base}
    rotation {yaw=-10}
   }
   n53 = "Transform (Modify) Bone - Bone: dyn_pouch_base"(ComponentPose=@n52, Alpha=1, Translation="0,0,0", Rotation="0.000,20,0.000", Scale="1,1,1") {
    rotationMode=BMM_Additive
    rotationSpace=BCS_BoneSpace
    translationMode=BMM_Additive
    translationSpace=BCS_BoneSpace
    alphaBoolBlend {blendOption=Linear}
    boneToModify {boneName=dyn_pouch_base}
    rotation {yaw=20}
   }
   n54 = "Transform (Modify) Bone - Bone: dyn_skirt_e_base"(ComponentPose=@n53, Alpha=1, Translation="0,0,0", Rotation="0,-20,0", Scale="1,1,1") {
    rotationMode=BMM_Additive
    rotationSpace=BCS_BoneSpace
    alphaBoolBlend {blendOption=Linear}
    boneToModify {boneName=dyn_skirt_e_base}
    rotation {yaw=-20}
   }
   n55 = "Transform (Modify) Bone - Bone: dyn_skirt_f_base"(ComponentPose=@n54, Alpha=1, Translation="0,0,0", Rotation="0,-10,0", Scale="1,1,1") {
    rotationMode=BMM_Additive
    rotationSpace=BCS_BoneSpace
    alphaBoolBlend {blendOption=Linear}
    boneToModify {boneName=dyn_skirt_f_base}
    rotation {yaw=-10}
   }
   n56 = Get bIsSkydiveDiveMode()
   n57 = "Transform (Modify) Bone - Bone: dyn_skirt_a_base"!(ComponentPose=@n55, bAlphaBoolEnabled:bool=@n56, Translation="0,0,0", Rotation="0,25,0", Scale="1,1,1") -> (Pose) {
    alphaInputType=Bool
    rotationMode=BMM_Additive
    rotationSpace=BCS_BoneSpace
    translationMode=BMM_Additive
    translationSpace=BCS_BoneSpace
    alphaBoolBlend {blendInTime=0.5, blendOption=Linear, blendOutTime=0.5}
    boneToModify {boneName=dyn_skirt_a_base}
    rotation {yaw=25}
   }
   n58 = "Transform (Modify) Bone - Bone: dyn_skirt_b_base"!(ComponentPose=@n57, bAlphaBoolEnabled:bool=@n56, Translation="0,0,0", Rotation="0,25,0", Scale="1,1,1") -> (Pose) {
    alphaInputType=Bool
    rotationMode=BMM_Additive
    rotationSpace=BCS_BoneSpace
    translationMode=BMM_Additive
    translationSpace=BCS_BoneSpace
    alphaBoolBlend {blendInTime=0.5, blendOption=Linear, blendOutTime=0.5}
    boneToModify {boneName=dyn_skirt_b_base}
    rotation {yaw=25}
   }
   n59 = "Transform (Modify) Bone - Bone: dyn_skirt_c_base"!(ComponentPose=@n58, bAlphaBoolEnabled:bool=@n56, Translation="0,0,0", Rotation="0,25,0", Scale="1,1,1") -> (Pose) {
    alphaInputType=Bool
    rotationMode=BMM_Additive
    rotationSpace=BCS_BoneSpace
    translationMode=BMM_Additive
    translationSpace=BCS_BoneSpace
    alphaBoolBlend {blendInTime=0.5, blendOption=Linear, blendOutTime=0.5}
    boneToModify {boneName=dyn_skirt_c_base}
    rotation {yaw=25}
   }
   n60 = "Transform (Modify) Bone - Bone: dyn_skirt_d_base"!(ComponentPose=@n59, bAlphaBoolEnabled:bool=@n56, Translation="0,0,0", Rotation="0,25,0", Scale="1,1,1") -> (Pose) {
    alphaInputType=Bool
    rotationMode=BMM_Additive
    rotationSpace=BCS_BoneSpace
    translationMode=BMM_Additive
    translationSpace=BCS_BoneSpace
    alphaBoolBlend {blendInTime=0.5, blendOption=Linear, blendOutTime=0.5}
    boneToModify {boneName=dyn_skirt_d_base}
    rotation {yaw=25}
   }
   n61 = "Transform (Modify) Bone - Bone: dyn_skirt_e_base"!(ComponentPose=@n60, bAlphaBoolEnabled:bool=@n56, Translation="0,0,0", Rotation="0,25,0", Scale="1,1,1") -> (Pose) {
    alphaInputType=Bool
    rotationMode=BMM_Additive
    rotationSpace=BCS_BoneSpace
    translationMode=BMM_Additive
    translationSpace=BCS_BoneSpace
    alphaBoolBlend {blendInTime=0.5, blendOption=Linear, blendOutTime=0.5}
    boneToModify {boneName=dyn_skirt_e_base}
    rotation {yaw=25}
   }
   n62 = "Transform (Modify) Bone - Bone: dyn_skirt02_e"!(ComponentPose=@n61, bAlphaBoolEnabled:bool=@n56, Translation="0,0,0", Rotation="0,15,0", Scale="1,1,1") -> (Pose) {
    alphaInputType=Bool
    rotationMode=BMM_Additive
    rotationSpace=BCS_BoneSpace
    translationMode=BMM_Additive
    translationSpace=BCS_BoneSpace
    alphaBoolBlend {blendInTime=0.5, blendOption=Linear, blendOutTime=0.5}
    boneToModify {boneName=dyn_skirt02_e}
    rotation {yaw=15}
   }
   n63 = "Transform (Modify) Bone - Bone: dyn_skirt2_f"!(ComponentPose=@n62, bAlphaBoolEnabled:bool=@n56, Translation="0,0,0", Rotation="0,25,0", Scale="1,1,1") -> (Pose) {
    alphaInputType=Bool
    rotationMode=BMM_Additive
    rotationSpace=BCS_BoneSpace
    translationMode=BMM_Additive
    translationSpace=BCS_BoneSpace
    alphaBoolBlend {blendInTime=0.5, blendOption=Linear, blendOutTime=0.5}
    boneToModify {boneName=dyn_skirt2_f}
    rotation {yaw=25}
   }
   n64 = Get thigh_r_rot_alpha()
   n65 = "Transform (Modify) Bone - Bone: dyn_skirt3_d"(ComponentPose=@n63, Alpha=@n64, Translation="0,0,0", Rotation="0.000,30,0.000", Scale="1,1,1") {
    rotationMode=BMM_Additive
    rotationSpace=BCS_BoneSpace
    translationMode=BMM_Additive
    translationSpace=BCS_BoneSpace
    alphaBoolBlend {blendOption=Linear}
    boneToModify {boneName=dyn_skirt3_d}
    rotation {yaw=30}
   }
   n66 = "Transform (Modify) Bone - Bone: dyn_skirt1_d"(ComponentPose=@n65, Alpha=@n64, Translation="0,0,0", Rotation="0.000,25,0.000", Scale="1,1,1") {
    rotationMode=BMM_Additive
    rotationSpace=BCS_BoneSpace
    translationMode=BMM_Additive
    translationSpace=BCS_BoneSpace
    alphaBoolBlend {blendOption=Linear}
    boneToModify {boneName=dyn_skirt1_d}
    rotation {yaw=25}
   }
   n67 = Get thigh_r_pitch_alpha()
   n68 = "Transform (Modify) Bone - Bone: dyn_skirt03_e"(ComponentPose=@n66, Alpha=@n67, Translation="-10,0,0", Rotation="-10.000,-20.000,20", Scale="1,1,1") {
    rotationMode=BMM_Additive
    rotationSpace=BCS_BoneSpace
    translationMode=BMM_Additive
    translationSpace=BCS_BoneSpace
    alphaBoolBlend {blendOption=Linear}
    boneToModify {boneName=dyn_skirt03_e}
    rotation=<rot -10 -20 20>
    translation {x=-10}
   }
   n69 = "Transform (Modify) Bone - Bone: dyn_skirt3_d"(ComponentPose=@n68, Alpha=@n67, Translation="0,0,0", Rotation="-20,0,0.000", Scale="1,1,1") {
    rotationMode=BMM_Additive
    rotationSpace=BCS_BoneSpace
    translationMode=BMM_Additive
    translationSpace=BCS_BoneSpace
    alphaBoolBlend {blendOption=Linear}
    boneToModify {boneName=dyn_skirt3_d}
    rotation {pitch=-20}
   }
   n70 = Get bIsCrouching()
   n71 = "Transform (Modify) Bone - Bone: dyn_skirt_a_base"(ComponentPose=@n69, bAlphaBoolEnabled=@n70, Translation="0,0,0", Rotation="0,0,0", Scale="1,1,1") {
    alphaInputType=Bool
    rotationMode=BMM_Additive
    rotationSpace=BCS_BoneSpace
    translationMode=BMM_Additive
    translationSpace=BCS_BoneSpace
    alphaBoolBlend {blendOption=Linear}
    boneToModify {boneName=dyn_skirt_a_base}
   }
   n72 = "Transform (Modify) Bone - Bone: dyn_skirt03_e"!(ComponentPose=@n71, bAlphaBoolEnabled:bool=@n70, Translation="0,0,0", Rotation="0,0,0", Scale="1,1,1") -> (Pose) {
    alphaInputType=Bool
    rotationMode=BMM_Additive
    rotationSpace=BCS_BoneSpace
    translationMode=BMM_Additive
    translationSpace=BCS_BoneSpace
    alphaBoolBlend {blendOption=Linear}
    boneToModify {boneName=dyn_skirt03_e}
   }
   n73 = "Transform (Modify) Bone - Bone: dyn_skirt2_f"(ComponentPose=@n72, bAlphaBoolEnabled=@n70, Translation="0,0,0", Rotation="0,-10,0", Scale="1,1,1") {
    alphaInputType=Bool
    rotationMode=BMM_Additive
    rotationSpace=BCS_BoneSpace
    translationMode=BMM_Additive
    translationSpace=BCS_BoneSpace
    alphaBoolBlend {blendOption=Linear}
    boneToModify {boneName=dyn_skirt2_f}
    rotation {yaw=-10}
   }
   n74 = "Transform (Modify) Bone - Bone: dyn_skirt04_e"(ComponentPose=@n73, Alpha=Get thigh_r_crouch_rot_alpha(), Translation="-2,0,0", Rotation="-25,0.000,0.000", Scale="1,1,1") {
    rotationMode=BMM_Additive
    rotationSpace=BCS_BoneSpace
    translationMode=BMM_Additive
    translationSpace=BCS_BoneSpace
    alphaBoolBlend {blendOption=Linear}
    boneToModify {boneName=dyn_skirt04_e}
    rotation {pitch=-25}
    translation {x=-2}
   }
   n75 = "Transform (Modify) Bone - Bone: dyn_skirt4_d"(ComponentPose=@n74, Alpha=1, Translation="0,0,0", Rotation="-25,20,0", Scale="1,1,1") {
    rotationMode=BMM_Additive
    rotationSpace=BCS_BoneSpace
    translationMode=BMM_Additive
    translationSpace=BCS_BoneSpace
    alphaBoolBlend {blendOption=Linear}
    boneToModify {boneName=dyn_skirt4_d}
    rotation {pitch=-25, yaw=20}
   }
   n76 = "Transform (Modify) Bone - Bone: pauldron_upperarm_base"(
    ComponentPose=@n75
    Alpha=Get upper_arm_l_rot_up_alpha() -> (upper_arm_l_rot_up_alpha:double)
    Translation="0,1,0"
    Rotation="45,35,-32"
    Scale="1,1,1"
   ) {
    rotationMode=BMM_Additive
    rotationSpace=BCS_BoneSpace
    scaleMode=BMM_Replace
    scaleSpace=BCS_BoneSpace
    translationMode=BMM_Additive
    translationSpace=BCS_BoneSpace
    alphaBoolBlend {blendOption=Linear}
    boneToModify {boneName=pauldron_upperarm_base}
    rotation=<rot 45 35 -32>
    translation {y=1}
   }
   n77 = "Transform (Modify) Bone - Bone: pauldron_upperarm_base"(
    ComponentPose=@n76
    Alpha=Get upperarm_l_pitch_alpha() -> (upperarm_l_pitch_alpha:double)
    Translation="0,0,0"
    Rotation="15.000,10,2.121"
    Scale="1,1,1"
   ) {
    rotationMode=BMM_Additive
    rotationSpace=BCS_BoneSpace
    scaleMode=BMM_Replace
    scaleSpace=BCS_BoneSpace
    translationMode=BMM_Additive
    translationSpace=BCS_BoneSpace
    alphaBoolBlend {blendOption=Linear}
    boneToModify {boneName=pauldron_upperarm_base}
    rotation=<rot 15 10 2.121>
   }
   n78 = "Transform (Modify) Bone - Bone: pauldron_upperarm_base"(
    ComponentPose=@n77
    Alpha=Get upper_arm_l_rot_down_alpha() -> (upper_arm_l_rot_down_alpha:double)
    Translation="0,0,1"
    Rotation="-15.000,10,0.000"
    Scale="1,1,1"
   ) {
    rotationMode=BMM_Additive
    rotationSpace=BCS_BoneSpace
    scaleMode=BMM_Replace
    scaleSpace=BCS_BoneSpace
    translationMode=BMM_Additive
    translationSpace=BCS_BoneSpace
    alphaBoolBlend {blendOption=Linear}
    boneToModify {boneName=pauldron_upperarm_base}
    rotation {pitch=-15, yaw=10}
    translation {z=1}
   }
   n79 = "Transform (Modify) Bone - Bone: pauldron_upperarm_base"(ComponentPose=@n78, Alpha=Get pickaxe_alpha() -> (pickaxe_alpha:double), Translation="0,-2,0", Rotation="15,15,0", Scale="1,1,1") {
    rotationMode=BMM_Additive
    rotationSpace=BCS_BoneSpace
    scaleMode=BMM_Replace
    scaleSpace=BCS_BoneSpace
    translationMode=BMM_Additive
    translationSpace=BCS_BoneSpace
    alphaBoolBlend {blendOption=Linear}
    boneToModify {boneName=pauldron_upperarm_base}
    rotation {pitch=15, yaw=15}
    translation {y=-2}
   }
   n80 = "Transform (Modify) Bone - Bone: pauldron_upperarm_base"(ComponentPose=@n79, Alpha=Get dual_pickaxe_alpha() -> (dual_pickaxe_alpha:double), Translation="0,0,0", Rotation="10,0,0", Scale="1,1,1") {
    rotationMode=BMM_Additive
    rotationSpace=BCS_BoneSpace
    scaleMode=BMM_Replace
    scaleSpace=BCS_BoneSpace
    translationMode=BMM_Additive
    translationSpace=BCS_BoneSpace
    alphaBoolBlend {blendOption=Linear}
    boneToModify {boneName=pauldron_upperarm_base}
    rotation {pitch=10}
   }
   n81 = "Transform (Modify) Bone - Bone: pauldron_upperarm_base"(ComponentPose=@n80, Alpha=Get 1h_pickaxe_alpha() -> (1h_pickaxe_alpha:double), Translation="-2,0,-3", Rotation="-30,-15,0", Scale="1,1,1") {
    rotationMode=BMM_Additive
    rotationSpace=BCS_BoneSpace
    scaleMode=BMM_Replace
    scaleSpace=BCS_BoneSpace
    translationMode=BMM_Additive
    translationSpace=BCS_BoneSpace
    alphaBoolBlend {blendOption=Linear}
    boneToModify {boneName=pauldron_upperarm_base}
    rotation {pitch=-30, yaw=-15}
    translation {x=-2, z=-3}
   }
   n82 = "Transform (Modify) Bone - Bone: pauldron_upperarm_base"(ComponentPose=@n81, Alpha=Get Glider Alpha() -> (Glider Alpha:double), Translation="2,0,3", Rotation="45,0,0", Scale="1,1,1") {
    rotationMode=BMM_Additive
    rotationSpace=BCS_BoneSpace
    scaleMode=BMM_Replace
    scaleSpace=BCS_BoneSpace
    translationMode=BMM_Additive
    translationSpace=BCS_BoneSpace
    alphaBoolBlend {blendOption=Linear}
    boneToModify {boneName=pauldron_upperarm_base}
    rotation {pitch=45}
    translation {x=2, z=3}
   }
   n83 = "Transform (Modify) Bone - Bone: sleeve"(
    ComponentPose=@n82
    Alpha=Get upper_arm_r_rot_alpha() -> (upper_arm_r_rot_alpha:double)
    Translation="0,-0.5,-1"
    Rotation="0.000,20,0.000"
    Scale="1,1,1"
   ) {
    rotationMode=BMM_Additive
    rotationSpace=BCS_BoneSpace
    scaleMode=BMM_Replace
    scaleSpace=BCS_BoneSpace
    translationMode=BMM_Additive
    translationSpace=BCS_BoneSpace
    alphaBoolBlend {blendOption=Linear}
    boneToModify {boneName=sleeve}
    rotation {yaw=20}
    translation {y=-0.5, z=-1}
   }
   n84 = "Transform (Modify) Bone - Bone: sleeve"(
    ComponentPose=@n83
    Alpha=Get upper_arm_r_rot_aim_alpha() -> (upper_arm_r_rot_aim_alpha:double)
    Translation="0,0.5,1"
    Rotation="0,-20,0"
    Scale="1,1,1"
   ) {
    rotationMode=BMM_Additive
    rotationSpace=BCS_BoneSpace
    scaleMode=BMM_Replace
    scaleSpace=BCS_BoneSpace
    translationMode=BMM_Additive
    translationSpace=BCS_BoneSpace
    alphaBoolBlend {blendOption=Linear}
    boneToModify {boneName=sleeve}
    rotation {yaw=-20}
    translation {y=0.5, z=1}
   }
   n85 = "Transform (Modify) Bone - Bone: sleeve"(ComponentPose=@n84, Alpha=1, Translation="0.15,0.859,1.276", Rotation="0,-10,0", Scale="1,1,1") {
    rotationMode=BMM_Additive
    rotationSpace=BCS_BoneSpace
    scaleMode=BMM_Replace
    scaleSpace=BCS_BoneSpace
    translationMode=BMM_Additive
    translationSpace=BCS_BoneSpace
    alphaBoolBlend {blendOption=Linear}
    boneToModify {boneName=sleeve}
    rotation {yaw=-10}
    translation=<0.15 0.859 1.276>
   }
   n86 = "Transform (Modify) Bone - Bone: dyn_skirt_belt_helper_01"!(ComponentPose=@n85, Alpha:float=1, Translation="0,0,0", Rotation="0,20,0", Scale="1,1,1") -> (Pose) {
    rotationMode=BMM_Additive
    rotationSpace=BCS_BoneSpace
    scaleMode=BMM_Replace
    scaleSpace=BCS_BoneSpace
    translationMode=BMM_Additive
    translationSpace=BCS_BoneSpace
    alphaBoolBlend {blendOption=Linear}
    boneToModify {boneName=dyn_skirt_belt_helper_01}
    rotation {yaw=20}
   }
   n87 = "Transform (Modify) Bone - Bone: dyn_skirt_a_base"!(ComponentPose=@n86, Alpha:float=1, Translation="0,0,0", Rotation="0,-20,0", Scale="1,1,1") -> (Pose) {
    rotationMode=BMM_Additive
    rotationSpace=BCS_BoneSpace
    scaleMode=BMM_Replace
    scaleSpace=BCS_BoneSpace
    translationMode=BMM_Additive
    translationSpace=BCS_BoneSpace
    alphaBoolBlend {blendOption=Linear}
    boneToModify {boneName=dyn_skirt_a_base}
    rotation {yaw=-20}
   }
   n88 = "Transform (Modify) Bone - Bone: dyn_skirt02_a"(ComponentPose=@n87, Alpha=1, Translation="0,2,0", Rotation="0,20,0", Scale="1,1,1") {
    rotationMode=BMM_Additive
    rotationSpace=BCS_BoneSpace
    scaleMode=BMM_Replace
    scaleSpace=BCS_BoneSpace
    translationMode=BMM_Additive
    translationSpace=BCS_BoneSpace
    alphaBoolBlend {blendOption=Linear}
    boneToModify {boneName=dyn_skirt02_a}
    rotation {yaw=20}
    translation {y=2}
   }
   n89 = RigidBody(ComponentPose=Local To Component(LocalPose=Copy Pose From Mesh() {bUseAttachedParent=true})) {
    bOverrideWorldGravity=true
    overridePhysicsAsset=&M_MED_Ninja_Wolf_WaistCloth_Physics
    simulationSpace=BaseBoneSpace
    alphaBoolBlend {blendOption=Linear}
    baseBoneRef {boneName=pelvis}
    componentAppliedLinearAccClamp=<2000 2000 2000>
    componentLinearAccScale=<0.3 0.3 0.35>
    componentLinearVelScale=<0.3 0.3 0.35>
    overrideWorldGravity {z=-980}
   }
   n90 = RigidBody(ComponentPose=@n89) {
    bOverrideWorldGravity=true
    overridePhysicsAsset=&M_MED_Ninja_Wolf_ArmL_Physics
    simulationSpace=BaseBoneSpace
    alphaBoolBlend {blendOption=Linear}
    baseBoneRef {boneName=lowerarm_l}
    componentAppliedLinearAccClamp=<1500 1500 1500>
    componentLinearAccScale=<0.25 0.25 0.3>
    componentLinearVelScale=<0.25 0.25 0.3>
    overrideWorldGravity {z=-980}
   }
   n91 = RigidBody(ComponentPose=@n90) {
    bOverrideWorldGravity=true
    overridePhysicsAsset=&M_MED_Ninja_Wolf_ArmR_Physics
    simulationSpace=BaseBoneSpace
    alphaBoolBlend {blendOption=Linear}
    baseBoneRef {boneName=lowerarm_r}
    componentAppliedLinearAccClamp=<1500 1500 1500>
    componentLinearAccScale=<0.25 0.25 0.3>
    componentLinearVelScale=<0.25 0.25 0.3>
    overrideWorldGravity {z=-980}
   }
   n92 = RigidBody(ComponentPose=@n91) {
    bOverrideWorldGravity=true
    overridePhysicsAsset=&M_MED_Ninja_Wolf_LegL_Physics
    simulationSpace=BaseBoneSpace
    alphaBoolBlend {blendOption=Linear}
    baseBoneRef {boneName=thigh_l}
    componentAppliedLinearAccClamp=<1500 1500 1500>
    componentLinearAccScale=<0.25 0.25 0.3>
    componentLinearVelScale=<0.25 0.25 0.3>
    overrideWorldGravity {z=-980}
   }
   n93 = RigidBody(ComponentPose=@n92) {
    bOverrideWorldGravity=true
    overridePhysicsAsset=&M_MED_Ninja_Wolf_LegR_Physics
    simulationSpace=BaseBoneSpace
    alphaBoolBlend {blendOption=Linear}
    baseBoneRef {boneName=thigh_r}
    componentAppliedLinearAccClamp=<1500 1500 1500>
    componentLinearAccScale=<0.25 0.25 0.3>
    componentLinearVelScale=<0.25 0.25 0.3>
    overrideWorldGravity {z=-980}
   }
   n94 = "Constraint - dyn_skirt01_a (PRR)"!(ComponentPose=@n93, Alpha=1, ConstraintWeights_0=0, ConstraintWeights_1:float=1, ConstraintWeights_2:float=0) -> (Pose) {
    alphaBoolBlend {blendOption=Linear}
    alphaScaleBiasClamp {bInterpResult=true}
    boneToModify {boneName=dyn_skirt01_a}
    constraintSetup [
     {targetBone {boneName=pelvis}}
     {transformType=Rotation, targetBone {boneName=thigh_twist_01_l}}
     {transformType=Rotation, targetBone {boneName=thigh_twist_01_r}}
    ]
    constraintWeights [1,1,1]
   }
   n95 = "Constraint - dyn_skirt02_a (PPP)"!(ComponentPose=@n94, Alpha=1, ConstraintWeights_0=0, ConstraintWeights_1:float=0.7, ConstraintWeights_2:float=0.3) -> (Pose) {
    alphaBoolBlend {blendOption=Linear}
    alphaScaleBiasClamp {bInterpResult=true}
    boneToModify {boneName=dyn_skirt02_a}
    constraintSetup [{targetBone {boneName=pelvis}}, {targetBone {boneName=thigh_twist_01_l}}, {targetBone {boneName=thigh_twist_01_r}}]
    constraintWeights [1,1,1]
   }
   n96 = "Constraint - dyn_skirt03_a (PPP)"!(ComponentPose=@n95, Alpha=1, ConstraintWeights_0=0, ConstraintWeights_1:float=0.7, ConstraintWeights_2:float=0.3) -> (Pose) {
    alphaBoolBlend {blendOption=Linear}
    alphaScaleBiasClamp {bInterpResult=true}
    boneToModify {boneName=dyn_skirt03_a}
    constraintSetup [{targetBone {boneName=pelvis}}, {targetBone {boneName=thigh_twist_01_l}}, {targetBone {boneName=thigh_twist_01_r}}]
    constraintWeights [1,1,1]
   }
   n97 = "Constraint - dyn_skirt1_b (PRR)"!(ComponentPose=@n96, Alpha=1, ConstraintWeights_0=0, ConstraintWeights_1:float=0, ConstraintWeights_2:float=1) -> (Pose) {
    alphaBoolBlend {blendOption=Linear}
    alphaScaleBiasClamp {bInterpResult=true}
    boneToModify {boneName=dyn_skirt1_b}
    constraintSetup [
     {targetBone {boneName=pelvis}}
     {transformType=Rotation, targetBone {boneName=thigh_twist_01_l}}
     {transformType=Rotation, targetBone {boneName=thigh_twist_01_r}}
    ]
    constraintWeights [1,1,1]
   }
   n98 = "Constraint - dyn_skirt2_b (PPP)"!(ComponentPose=@n97, Alpha=1, ConstraintWeights_0=0, ConstraintWeights_1:float=0, ConstraintWeights_2:float=1) -> (Pose) {
    alphaBoolBlend {blendOption=Linear}
    alphaScaleBiasClamp {bInterpResult=true}
    boneToModify {boneName=dyn_skirt2_b}
    constraintSetup [{targetBone {boneName=pelvis}}, {targetBone {boneName=thigh_twist_01_l}}, {targetBone {boneName=thigh_twist_01_r}}]
    constraintWeights [1,1,1]
   }
   n99 = "Constraint - dyn_skirt3_b (PPP)"!(ComponentPose=@n98, Alpha=1, ConstraintWeights_0=0, ConstraintWeights_1:float=0, ConstraintWeights_2:float=1) -> (Pose) {
    alphaBoolBlend {blendOption=Linear}
    alphaScaleBiasClamp {bInterpResult=true}
    boneToModify {boneName=dyn_skirt3_b}
    constraintSetup [{targetBone {boneName=pelvis}}, {targetBone {boneName=thigh_twist_01_l}}, {targetBone {boneName=thigh_twist_01_r}}]
    constraintWeights [1,1,1]
   }
   n100 = "Constraint - dyn_skirt01_c (PRR)"!(ComponentPose=@n99, Alpha=1, ConstraintWeights_0=0, ConstraintWeights_1:float=0, ConstraintWeights_2:float=1) -> (Pose) {
    alphaBoolBlend {blendOption=Linear}
    alphaScaleBiasClamp {bInterpResult=true}
    boneToModify {boneName=dyn_skirt01_c}
    constraintSetup [
     {targetBone {boneName=pelvis}}
     {transformType=Rotation, targetBone {boneName=thigh_twist_01_l}}
     {transformType=Rotation, targetBone {boneName=thigh_twist_01_r}}
    ]
    constraintWeights [1,1,1]
   }
   n101 = "Constraint - dyn_skirt02_c (PPP)"!(ComponentPose=@n100, Alpha=1, ConstraintWeights_0=0, ConstraintWeights_1:float=0, ConstraintWeights_2:float=1) -> (Pose) {
    alphaBoolBlend {blendOption=Linear}
    alphaScaleBiasClamp {bInterpResult=true}
    boneToModify {boneName=dyn_skirt02_c}
    constraintSetup [{targetBone {boneName=pelvis}}, {targetBone {boneName=thigh_twist_01_l}}, {targetBone {boneName=thigh_twist_01_r}}]
    constraintWeights [1,1,1]
   }
   n102 = "Constraint - dyn_skirt03_c (PPP)"!(ComponentPose=@n101, Alpha=1, ConstraintWeights_0=0, ConstraintWeights_1:float=0, ConstraintWeights_2:float=1) -> (Pose) {
    alphaBoolBlend {blendOption=Linear}
    alphaScaleBiasClamp {bInterpResult=true}
    boneToModify {boneName=dyn_skirt03_c}
    constraintSetup [{targetBone {boneName=pelvis}}, {targetBone {boneName=thigh_twist_01_l}}, {targetBone {boneName=thigh_twist_01_r}}]
    constraintWeights [1,1,1]
   }
   n103 = "Constraint - dyn_skirt04_c (PPP)"!(ComponentPose=@n102, Alpha=1, ConstraintWeights_0=0, ConstraintWeights_1:float=0, ConstraintWeights_2:float=1) -> (Pose) {
    alphaBoolBlend {blendOption=Linear}
    alphaScaleBiasClamp {bInterpResult=true}
    boneToModify {boneName=dyn_skirt04_c}
    constraintSetup [{targetBone {boneName=pelvis}}, {targetBone {boneName=thigh_twist_01_l}}, {targetBone {boneName=thigh_twist_01_r}}]
    constraintWeights [1,1,1]
   }
   n104 = "Constraint - dyn_skirt1_d (PRR)"!(ComponentPose=@n103, Alpha=1, ConstraintWeights_0=0, ConstraintWeights_1:float=0, ConstraintWeights_2:float=1) -> (Pose) {
    alphaBoolBlend {blendOption=Linear}
    alphaScaleBiasClamp {bInterpResult=true}
    boneToModify {boneName=dyn_skirt1_d}
    constraintSetup [
     {targetBone {boneName=pelvis}}
     {transformType=Rotation, targetBone {boneName=thigh_twist_01_l}}
     {transformType=Rotation, targetBone {boneName=thigh_twist_01_r}}
    ]
    constraintWeights [1,1,1]
   }
   n105 = "Constraint - dyn_skirt2_d (PPP)"!(ComponentPose=@n104, Alpha=1, ConstraintWeights_0=0, ConstraintWeights_1:float=0, ConstraintWeights_2:float=1) -> (Pose) {
    alphaBoolBlend {blendOption=Linear}
    alphaScaleBiasClamp {bInterpResult=true}
    boneToModify {boneName=dyn_skirt2_d}
    constraintSetup [{targetBone {boneName=pelvis}}, {targetBone {boneName=thigh_twist_01_l}}, {targetBone {boneName=thigh_twist_01_r}}]
    constraintWeights [1,1,1]
   }
   n106 = "Constraint - dyn_skirt3_d (PPP)"!(ComponentPose=@n105, Alpha=1, ConstraintWeights_0=0, ConstraintWeights_1:float=0, ConstraintWeights_2:float=1) -> (Pose) {
    alphaBoolBlend {blendOption=Linear}
    alphaScaleBiasClamp {bInterpResult=true}
    boneToModify {boneName=dyn_skirt3_d}
    constraintSetup [{targetBone {boneName=pelvis}}, {targetBone {boneName=thigh_twist_01_l}}, {targetBone {boneName=thigh_twist_01_r}}]
    constraintWeights [1,1,1]
   }
   n107 = "Constraint - dyn_skirt4_d (PPP)"!(ComponentPose=@n106, Alpha=1, ConstraintWeights_0=0, ConstraintWeights_1:float=0, ConstraintWeights_2:float=1) -> (Pose) {
    alphaBoolBlend {blendOption=Linear}
    alphaScaleBiasClamp {bInterpResult=true}
    boneToModify {boneName=dyn_skirt4_d}
    constraintSetup [{targetBone {boneName=pelvis}}, {targetBone {boneName=thigh_twist_01_l}}, {targetBone {boneName=thigh_twist_01_r}}]
    constraintWeights [1,1,1]
   }
   n108 = "Constraint - dyn_skirt02_e (PPP)"!(ComponentPose=@n107, Alpha=1, ConstraintWeights_0=0, ConstraintWeights_1:float=0, ConstraintWeights_2:float=1) -> (Pose) {
    alphaBoolBlend {blendOption=Linear}
    alphaScaleBiasClamp {bInterpResult=true}
    boneToModify {boneName=dyn_skirt02_e}
    constraintSetup [{targetBone {boneName=pelvis}}, {targetBone {boneName=thigh_twist_01_l}}, {targetBone {boneName=thigh_twist_01_r}}]
    constraintWeights [1,1,1]
   }
   n109 = "Constraint - dyn_skirt03_e (PPP)"!(ComponentPose=@n108, Alpha=1, ConstraintWeights_0=0, ConstraintWeights_1:float=0, ConstraintWeights_2:float=1) -> (Pose) {
    alphaBoolBlend {blendOption=Linear}
    alphaScaleBiasClamp {bInterpResult=true}
    boneToModify {boneName=dyn_skirt03_e}
    constraintSetup [{targetBone {boneName=pelvis}}, {targetBone {boneName=thigh_twist_01_l}}, {targetBone {boneName=thigh_twist_01_r}}]
    constraintWeights [1,1,1]
   }
   n110 = "Constraint - dyn_skirt04_e (PPP)"!(ComponentPose=@n109, Alpha=1, ConstraintWeights_0=0, ConstraintWeights_1:float=0, ConstraintWeights_2:float=1) -> (Pose) {
    alphaBoolBlend {blendOption=Linear}
    alphaScaleBiasClamp {bInterpResult=true}
    boneToModify {boneName=dyn_skirt04_e}
    constraintSetup [{targetBone {boneName=pelvis}}, {targetBone {boneName=thigh_twist_01_l}}, {targetBone {boneName=thigh_twist_01_r}}]
    constraintWeights [1,1,1]
   }
   n111 = "Constraint - dyn_skirt2_f (PPP)"!(ComponentPose=@n110, Alpha=1, ConstraintWeights_0=0, ConstraintWeights_1:float=0.5, ConstraintWeights_2:float=0.5) -> (Pose) {
    alphaBoolBlend {blendOption=Linear}
    alphaScaleBiasClamp {bInterpResult=true}
    boneToModify {boneName=dyn_skirt2_f}
    constraintSetup [{targetBone {boneName=pelvis}}, {targetBone {boneName=thigh_twist_01_l}}, {targetBone {boneName=thigh_twist_01_r}}]
    constraintWeights [1,1,1]
   }
   n112 = "Constraint - dyn_skirt3_f (PPP)"!(ComponentPose=@n111, Alpha=1, ConstraintWeights_0=0, ConstraintWeights_1:float=0.4, ConstraintWeights_2:float=0.6) -> (Pose) {
    alphaBoolBlend {blendOption=Linear}
    alphaScaleBiasClamp {bInterpResult=true}
    boneToModify {boneName=dyn_skirt3_f}
    constraintSetup [{targetBone {boneName=pelvis}}, {targetBone {boneName=thigh_twist_01_l}}, {targetBone {boneName=thigh_twist_01_r}}]
    constraintWeights [1,1,1]
   }
   n113 = "Constraint - dyn_pouch_base (PPP)"!(ComponentPose=@n112, Alpha=1, ConstraintWeights_0=0, ConstraintWeights_1:float=0, ConstraintWeights_2:float=1) -> (Pose) {
    alphaBoolBlend {blendOption=Linear}
    alphaScaleBiasClamp {bInterpResult=true}
    boneToModify {boneName=dyn_pouch_base}
    constraintSetup [{targetBone {boneName=pelvis}}, {targetBone {boneName=thigh_twist_01_l}}, {targetBone {boneName=thigh_twist_01_r}}]
    constraintWeights [1,1,1]
   }
   n114 = Look At - Bone: dyn_skirt03_a(ComponentPose=@n113, Alpha=1) {
    alphaBoolBlend {blendOption=Linear}
    alphaScaleBiasClamp {bInterpResult=true}
    boneToModify {boneName=dyn_skirt03_a}
    lookAtLocation=<-4 -18 25>
    lookAtTarget {boneReference {boneName=dyn_skirt3_b}}
    lookAt_Axis {axis {y=0, z=1}}
   }
   n115 = Look At - Bone: dyn_skirt3_b(ComponentPose=@n114, Alpha=1) {
    alphaBoolBlend {blendOption=Linear}
    alphaScaleBiasClamp {bInterpResult=true}
    boneToModify {boneName=dyn_skirt3_b}
    lookAtLocation=<4 -25 -25>
    lookAtTarget {boneReference {boneName=dyn_skirt03_a}}
    lookAt_Axis {axis {y=0, z=-1}}
   }
   n116 = "Transform (Modify) Bone - Bone: dyn_skirt_a_base"(ComponentPose=@n115, Alpha=1, Translation="0,-1,0", Rotation="0,0,0", Scale="1,1,1") {
    rotationMode=BMM_Additive
    rotationSpace=BCS_BoneSpace
    translationMode=BMM_Additive
    translationSpace=BCS_BoneSpace
    alphaBoolBlend {blendOption=Linear}
    alphaScaleBiasClamp {bInterpResult=true}
    boneToModify {boneName=dyn_skirt_a_base}
    translation {y=-1}
   }
   n117 = "Transform (Modify) Bone - Bone: dyn_skirt_b_base"!(ComponentPose=@n116, Alpha:float=1, Translation="0,-1,0", Rotation="0,0,0", Scale="1,1,1") -> (Pose) {
    rotationMode=BMM_Additive
    rotationSpace=BCS_BoneSpace
    translationMode=BMM_Additive
    translationSpace=BCS_BoneSpace
    alphaBoolBlend {blendOption=Linear}
    alphaScaleBiasClamp {bInterpResult=true}
    boneToModify {boneName=dyn_skirt_b_base}
    translation {y=-1}
   }
   n118 = "Transform (Modify) Bone - Bone: dyn_skirt_c_base"!(ComponentPose=@n117, Alpha:float=1, Translation="0,-1,0", Rotation="0,0,0", Scale="1,1,1") -> (Pose) {
    rotationMode=BMM_Additive
    rotationSpace=BCS_BoneSpace
    translationMode=BMM_Additive
    translationSpace=BCS_BoneSpace
    alphaBoolBlend {blendOption=Linear}
    alphaScaleBiasClamp {bInterpResult=true}
    boneToModify {boneName=dyn_skirt_c_base}
    translation {y=-1}
   }
   n119 = "Transform (Modify) Bone - Bone: dyn_skirt_d_base"!(ComponentPose=@n118, Alpha:float=1, Translation="0,-1,0", Rotation="0,0,0", Scale="1,1,1") -> (Pose) {
    rotationMode=BMM_Additive
    rotationSpace=BCS_BoneSpace
    translationMode=BMM_Additive
    translationSpace=BCS_BoneSpace
    alphaBoolBlend {blendOption=Linear}
    alphaScaleBiasClamp {bInterpResult=true}
    boneToModify {boneName=dyn_skirt_d_base}
    translation {y=-1}
   }
   n120 = "Transform (Modify) Bone - Bone: dyn_skirt_e_base"!(ComponentPose=@n119, Alpha:float=1, Translation="0,-1,0", Rotation="0,0,0", Scale="1,1,1") -> (Pose) {
    rotationMode=BMM_Additive
    rotationSpace=BCS_BoneSpace
    translationMode=BMM_Additive
    translationSpace=BCS_BoneSpace
    alphaBoolBlend {blendOption=Linear}
    alphaScaleBiasClamp {bInterpResult=true}
    boneToModify {boneName=dyn_skirt_e_base}
    translation {y=-1}
   }
   n121 = "Transform (Modify) Bone - Bone: dyn_skirt_f_base"(ComponentPose=@n120, Alpha=1, Translation="0,-1,0", Rotation="0,-10,0", Scale="1,1,1") {
    rotationMode=BMM_Additive
    rotationSpace=BCS_BoneSpace
    translationMode=BMM_Additive
    translationSpace=BCS_BoneSpace
    alphaBoolBlend {blendOption=Linear}
    alphaScaleBiasClamp {bInterpResult=true}
    boneToModify {boneName=dyn_skirt_f_base}
    rotation {yaw=-10}
    translation {y=-1}
   }
   n122 = "Transform (Modify) Bone - Bone: dyn_pouch_base"(ComponentPose=@n121, Alpha=1, Translation="0,0,0", Rotation="0,-5,0", Scale="1,1,1") {
    rotationMode=BMM_Additive
    rotationSpace=BCS_BoneSpace
    translationMode=BMM_Additive
    translationSpace=BCS_BoneSpace
    alphaBoolBlend {blendOption=Linear}
    alphaScaleBiasClamp {bInterpResult=true}
    boneToModify {boneName=dyn_pouch_base}
    rotation {yaw=-5}
   }
   n123 = "Transform (Modify) Bone - Bone: dyn_pouch_base"(ComponentPose=@n122, Alpha=1, Translation="15,1,0", Rotation="0,0,0", Scale="1,1,1") {
    rotationMode=BMM_Additive
    rotationSpace=BCS_BoneSpace
    translationMode=BMM_Additive
    translationSpace=BCS_BoneSpace
    alphaBoolBlend {blendOption=Linear}
    boneToModify {boneName=dyn_pouch_base}
    translation=<15 1>
   }
   n124 = "Transform (Modify) Bone - Bone: dyn_skirt_a_base"(ComponentPose=@n123, Alpha=1, Translation="0,0,0", Rotation="0,-5,0", Scale="1,1,1") {
    rotationMode=BMM_Additive
    rotationSpace=BCS_BoneSpace
    translationMode=BMM_Additive
    translationSpace=BCS_BoneSpace
    alphaBoolBlend {blendOption=Linear}
    boneToModify {boneName=dyn_skirt_a_base}
    rotation {yaw=-5}
   }
   n125 = "Transform (Modify) Bone - Bone: dyn_skirt_b_base"(ComponentPose=@n124, Alpha=1, Translation="0,0,0", Rotation="0,-5,0", Scale="1,1,1") {
    rotationMode=BMM_Additive
    rotationSpace=BCS_BoneSpace
    translationMode=BMM_Additive
    translationSpace=BCS_BoneSpace
    alphaBoolBlend {blendOption=Linear}
    boneToModify {boneName=dyn_skirt_b_base}
    rotation {yaw=-5}
   }
   n126 = "Transform (Modify) Bone - Bone: dyn_skirt2_b"(ComponentPose=@n125, Alpha=1, Translation="0,0,0", Rotation="0,20,0", Scale="1,1,1") {
    rotationMode=BMM_Additive
    rotationSpace=BCS_BoneSpace
    translationMode=BMM_Additive
    translationSpace=BCS_BoneSpace
    alphaBoolBlend {blendOption=Linear}
    boneToModify {boneName=dyn_skirt2_b}
    rotation {yaw=20}
   }
   n127 = "Transform (Modify) Bone - Bone: dyn_skirt_c_base"(ComponentPose=@n126, Alpha=1, Translation="0,0,0", Rotation="0,0,0", Scale="1,1,1") {
    rotationMode=BMM_Additive
    rotationSpace=BCS_BoneSpace
    translationMode=BMM_Additive
    translationSpace=BCS_BoneSpace
    alphaBoolBlend {blendOption=Linear}
    boneToModify {boneName=dyn_skirt_c_base}
   }
   n128 = "Transform (Modify) Bone - Bone: dyn_skirt_d_base"(ComponentPose=@n127, Alpha=1, Translation="0,0,0", Rotation="0,-5,0", Scale="1,1,1") {
    rotationMode=BMM_Additive
    rotationSpace=BCS_BoneSpace
    translationMode=BMM_Additive
    translationSpace=BCS_BoneSpace
    alphaBoolBlend {blendOption=Linear}
    boneToModify {boneName=dyn_skirt_d_base}
    rotation {yaw=-5}
   }
   n129 = "Transform (Modify) Bone - Bone: dyn_skirt_e_base"(ComponentPose=@n128, Alpha=1, Translation="0,0,0", Rotation="0,-5,0.000", Scale="1,1,1") {
    rotationMode=BMM_Additive
    rotationSpace=BCS_BoneSpace
    translationMode=BMM_Additive
    translationSpace=BCS_BoneSpace
    alphaBoolBlend {blendOption=Linear}
    boneToModify {boneName=dyn_skirt_e_base}
    rotation {yaw=-5}
   }
   n130 = "Transform (Modify) Bone - Bone: dyn_skirt03_e"!(ComponentPose=@n129, Alpha:float=1, Translation="0,0,0", Rotation="-10,0,0.000", Scale="1,1,1") -> (Pose) {
    rotationMode=BMM_Additive
    rotationSpace=BCS_BoneSpace
    translationMode=BMM_Additive
    translationSpace=BCS_BoneSpace
    alphaBoolBlend {blendOption=Linear}
    boneToModify {boneName=dyn_skirt03_e}
    rotation {pitch=-10}
   }
   n131 = "Transform (Modify) Bone - Bone: dyn_pouch_base"(ComponentPose=@n130, Alpha=1, Translation="0,0,0", Rotation="0.000,8,0.000", Scale="1,1,1") {
    rotationMode=BMM_Additive
    rotationSpace=BCS_BoneSpace
    translationMode=BMM_Additive
    translationSpace=BCS_BoneSpace
    alphaBoolBlend {blendOption=Linear}
    boneToModify {boneName=dyn_pouch_base}
    rotation {yaw=8}
   }
   Output Pose(Result=Blend Poses by bool(BlendPose_0=Component To Local(ComponentPose=@n131), BlendPose_1=Component To Local(ComponentPose=@n88), bActiveValue=Get bIsInFrontEnd()) {blendType=Linear}) {layerGroup=None}
  ]}
 EventGraph {nodes [
   Event BlueprintUpdateAnimation()
   > n1 = Sequence!(execute) -> (then_0, then_1, then_2:Exec)
   n2 = Get bIsSkydiveDiveMode()
   > n3 = Branch(Condition=OR Boolean!(A=Get bIsSkydiveFloating(), B=Get bIsParachuteOpened(), C:bool=@n2, D:bool=@n2) -> (ReturnValue))
   > Set cloth_alpha(cloth_alpha:double=0.5) -> (Output_Get:double)
   > Set Glider Alpha(Glider Alpha:double=1) -> (Output_Get:double)
   > Set upper_arm_l_rot_up_alpha(upper_arm_l_rot_up_alpha:double=0) -> (Output_Get:double)
   > Set upper_arm_l_rot_down_alpha(upper_arm_l_rot_down_alpha:double=0) -> (Output_Get:double)
   > Set upperarm_l_pitch_alpha(upperarm_l_pitch_alpha:double=0) -> (Output_Get:double)
   > Set upper_arm_r_rot_alpha(upper_arm_r_rot_alpha:double=0) -> (Output_Get:double)
   > Set upper_arm_r_rot_aim_alpha(upper_arm_r_rot_aim_alpha:double=0) -> (Output_Get:double)
   Set cloth_alpha(execute=@n3.else, cloth_alpha=0)
   > Set Glider Alpha(Glider Alpha=0)
   n4 = GetOwnerSkeletalMeshComponent()
   n5 = BreakRotator(InRot=BreakTransform(InTransform=GetDeltaTransformFromRefPose(self=@n4, BoneName=upperarm_l, BaseName=clavicle_l)).Rotation)
   > Set upper_arm_l_rot_up_alpha(upper_arm_l_rot_up_alpha=MapRangeClamped(Value=@n5.Yaw, InRangeA=0, InRangeB=90))
   > Set upper_arm_l_rot_down_alpha(upper_arm_l_rot_down_alpha=MapRangeClamped(Value=@n5, InRangeA=-50, InRangeB=0))
   > Set upperarm_l_pitch_alpha(upperarm_l_pitch_alpha=MapRangeClamped(Value=@n5.Pitch, InRangeA=30, InRangeB=50))
   n6 = BreakRotator(InRot=BreakTransform(InTransform=GetDeltaTransformFromRefPose(self=@n4, BoneName=upperarm_r, BaseName=clavicle_r)).Rotation)
   > Set upper_arm_r_rot_alpha(upper_arm_r_rot_alpha=MapRangeClamped(Value=@n6.Yaw, InRangeA=0, InRangeB=40))
   > Set upper_arm_r_rot_aim_alpha(upper_arm_r_rot_aim_alpha=MapRangeClamped(Value=@n6.Yaw, InRangeA=50, InRangeB=100))
   > Set pickaxe_alpha(pickaxe_alpha:double=0) -> (Output_Get:double)
   > Set dual_pickaxe_alpha(dual_pickaxe_alpha:double=0) -> (Output_Get:double)
   > Set 1h_pickaxe_alpha(1h_pickaxe_alpha:double=0) -> (Output_Get:double)
   > n7 = Cast To FortPlayerAnimInstance(Object=GetAnimInstance(self=GetOwnerSkeletalMeshComponent()))
   > n8 = Switch on EFortWeaponCoreAnimation!(execute, Selection:EFortWeaponCoreAnimation Enum=Get WeaponCoreAnim(self=@n7)) -> (
    Melee:Exec
    Pistol:Exec
    Shotgun:Exec
    PaperBlueprint:Exec
    Rifle:Exec
    MeleeOneHand:Exec
    MachinePistol:Exec
    RocketLauncher:Exec
    GrenadeLauncher:Exec
    GoingCommando:Exec
    AssaultRifle:Exec
    TacticalShotgun:Exec
    SniperRifle:Exec
    TrapPlacement:Exec
    ShoulderLauncher:Exec
    AbilityDecoTool:Exec
    Crossbow:Exec
    C4:Exec
    RemoteControl:Exec
    DualWield:Exec
    AR_BullPup:Exec
    AR_ForwardGrip:Exec
    MedPackPaddles:Exec
    SMG_P90:Exec
    AR_DrumGun:Exec
    Consumable_Small:Exec
    Consumable_Large:Exec
    Balloon:Exec
    MountedTurret:Exec
    CreativeTool:Exec
    ExplosiveBow:Exec
    AshtonIndigo:Exec
    AshtonChicago:Exec
    MeleeDualWield:Exec
    Unarmed:Exec
   )
   > Set upper_arm_l_rot_up_alpha(upper_arm_l_rot_up_alpha=0)
   > Set upperarm_l_pitch_alpha(upperarm_l_pitch_alpha=0)
   > Set upper_arm_l_rot_down_alpha(upper_arm_l_rot_down_alpha=0)
   n9 = BreakTransform(InTransform=GetDeltaTransformFromRefPose(self=GetOwnerSkeletalMeshComponent(), BoneName=pelvis, BaseName=root))
   > Set pickaxe_alpha(pickaxe_alpha=MapRangeClamped(Value=BreakRotator(InRot=@n9.Rotation).Yaw, InRangeA=0, InRangeB=15))
   Set dual_pickaxe_alpha(execute=@n8.MeleeDualWield, dual_pickaxe_alpha=1)
   Set 1h_pickaxe_alpha(execute=@n8.MeleeOneHand, 1h_pickaxe_alpha=1)
   n10 = Branch(execute=@n1.then_1, Condition=Get bIsCrouching())
   > Set thigh_r_pitch_alpha(thigh_r_pitch_alpha:double=0) -> (Output_Get:double)
   > Set thigh_r_rot_alpha(thigh_r_rot_alpha:double=0) -> (Output_Get:double)
   n11 = GetOwnerSkeletalMeshComponent()
   n12 = BreakRotator(InRot=BreakTransform(InTransform=GetDeltaTransformFromRefPose(self=@n11, BoneName=thigh_r, BaseName=pelvis)).Rotation)
   > Set thigh_r_crouch_rot_roll_alpha(thigh_r_crouch_rot_roll_alpha:double=MapRangeClamped(Value=@n12, InRangeA=0, InRangeB=45)) -> (Output_Get:double)
   > Set thigh_r_crouch_pitch_alpha(thigh_r_crouch_pitch_alpha:double=MapRangeClamped(Value=@n12.Pitch, InRangeA=0, InRangeB=40)) -> (Output_Get:double)
   > Set thigh_r_crouch_rot_alpha(thigh_r_crouch_rot_alpha:double=MapRangeClamped(Value=@n12.Yaw, InRangeA=-25, InRangeB=-105)) -> (Output_Get:double)
   Set thigh_r_crouch_rot_alpha(execute=@n10.else, thigh_r_crouch_rot_alpha=0)
   > Set thigh_r_crouch_rot_roll_alpha(thigh_r_crouch_rot_roll_alpha=0)
   > Set thigh_r_crouch_pitch_alpha(thigh_r_crouch_pitch_alpha=0)
   n13 = GetOwnerSkeletalMeshComponent()
   n14 = BreakRotator(InRot=BreakTransform(InTransform=GetDeltaTransformFromRefPose(self=@n13, BoneName=thigh_r, BaseName=pelvis)).Rotation)
   > Set thigh_r_pitch_alpha(thigh_r_pitch_alpha=MapRangeClamped(Value=@n14.Pitch, InRangeA=0, InRangeB=45))
   > Set thigh_r_rot_alpha(thigh_r_rot_alpha=MapRangeClamped(Value=@n14.Yaw, InRangeA=0, InRangeB=-90))
   > Set jump_falling(jump_falling:bool=OR Boolean!(A=Get bIsJumping(), B=Get bIsFalling()) -> (ReturnValue)) -> (Output_Get:bool)
   n15 = Branch(execute=@n1.then_2, Condition=Get jump_falling() -> (jump_falling:bool))
   n16 = BreakTransform(InTransform=GetDeltaTransformFromRefPose(self=GetOwnerSkeletalMeshComponent(), BoneName=thigh_r, BaseName=pelvis))
   > Set jump_alpha(jump_alpha:double=MapRangeClamped(Value=BreakRotator(InRot=@n16.Rotation).Yaw, InRangeA=0, InRangeB=-70)) -> (Output_Get:double)
   Set jump_alpha(execute=@n15.else, jump_alpha=0)
   MapRangeClamped(Value=@n14, InRangeA=0, InRangeB=45)
  ]}
]
variables [
 1h_pickaxe_alpha {type=number}
 cloth_alpha {type=number}
 dual_pickaxe_alpha {type=number}
 Glider Alpha {type=number, default=1}
 jump_alpha {type=number}
 jump_falling {type=boolean}
 pickaxe_alpha {type=number}
 thigh_r_crouch_pitch_alpha {type=number}
 thigh_r_crouch_rot_alpha {type=number}
 thigh_r_crouch_rot_roll_alpha {type=number}
 thigh_r_pitch_alpha {type=number}
 thigh_r_rot_alpha {type=number}
 upper_arm_l_rot_down_alpha {type=number}
 upper_arm_l_rot_up_alpha {type=number}
 upper_arm_r_rot_aim_alpha {type=number}
 upper_arm_r_rot_alpha {type=number}
 upperarm_l_pitch_alpha {type=number}
]
